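# Translucent unlit material for mesh particles, presumably a glowing cylinder used in bling effects.
# Emissive tints particle color with a time-driven sine over tiled cloud noise, lerping DarkColor to white, plus a gradient band slid by the GlowPan dynamic parameter.
# Opacity combines particle alpha, that band, a vertical mask, view-angle fade and axis-aligned Fresnel.
# Quality switch drops camera and depth fades on Low and Medium.
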Material /Game/Effects/Fort_Effects/Effects/Bling/M_Bling_Cylinder
graph {
 properties {
  bUsedWithMeshParticles=true
  bUsedWithNiagaraMeshParticles=true
  blendMode=BLEND_Translucent
  pixelDepthOffsetMode=PDOM_Legacy
  shadingModel=MSM_Unlit
  usedShadingModels=Unlit
 }
 nodes [
  n1 = ParticleColor()
  n2 = VectorParameter() {parameterName=UVScale, defaultValue=<rgba 1 0.01 0 1>}
  n3 = Multiply(
   A=Multiply(A=ComponentMask(None=@n2.RGB) {g=true, r=true}, B=TextureCoordinate() {vTiling=2})
   B=ScalarParameter() {defaultValue=3, parameterName=SquareTile}
  )
  n4 = TextureSample(UVs=@n3) {samplerType=SAMPLERTYPE_LinearColor, texture=/Game/Global/Textures/Default/Tileable_Patterns/Cloud_Diffuse}
  n5 = Multiply(A=Time()) {constB=1.5}
  n6 = LinearInterpolate(
   A=VectorParameter() {parameterName=DarkColor, defaultValue=<rgba 0.1178 0.6628 1 1>}.RGB
   Alpha=MaterialFunctionCall!(Sine Phase=Add(A=@n4.R, B=@n5), Value 1, Value 2) {materialFunction=/Engine/Functions/Engine_MaterialFunctions02/Utility/Sine_Remapped}.Remapped Sine
  ) {constA=1}
  n7 = TextureCoordinate() {vTiling=6}
  n8 = TextureCoordinate()
  n9 = TextureSample(UVs=LinearInterpolate(
    A=Add(A=@n7) {constB=-5.5}
    B=Add(A=@n8) {constB=0.7}
    Alpha=DynamicParameter() {defaultValue=<rgba 1 1 0 1>, paramNames [FresnelExponent, Refraction, GlowPan, Param4]}.GlowPan
   )) {texture=/Game/Global/Textures/Default/Masks_and_Gradients/T_Vert_Gradient_Soft}
  n10 = Multiply(A=Power(Base=@n9.R) {constExponent=3}) {constB=2}
  MaterialFunctionCall(Emissive=Multiply(A=@n1.RGB, B=Add(A=@n6, B=@n10) {constA=1})) {materialFunction=&MF_FX_EmissiveAdjustments}
  Panner(Coordinate=@n3) {speedY=0.25}
  n11 = ComponentMask(None=TextureCoordinate()) {g=true}
  n12 = Multiply(
   A=Add(A=@n9.R) {constB=0.1}
   B=TextureSample(UVs=OneMinus(None=@n11)) {texture=/Game/Effects/Fort_Effects/Textures/Masks/T_Vert_Gradient_Softer}.R
  )
  n13 = MaterialFunctionCall("FadeSensitivity (optional)"=ScalarParameter() {defaultValue=2, parameterName=FadeAngle}) {materialFunction=&FadeBasedOnViewAngle}
  n14 = Multiply(
   A=Multiply(A=Multiply(A=@n1.A, B=@n12), B=LinearInterpolate(Alpha=@n13.Result) {constA=1, constB=0.1})
   B=Power(Base=MaterialFunctionCall() {materialFunction=&AxisAlignedFresnel}.Fresnel)
  )
  n15 = Multiply(
   A=MaterialFunctionCall(
    Fade Length=ScalarParameter() {defaultValue=512, parameterName="Fade Length (S)"}
    Fade Offset=ScalarParameter() {defaultValue=1024, parameterName="Fade Offset (S)"}
   ) {materialFunction=&CameraDepthFade}.Result
   B=DepthFade(Opacity=@n14) {fadeDistanceDefault=15}
  )
  QualitySwitch(Default=@n15, Low=@n14, High=@n15, Medium=@n14, Epic=@n15)
 ]
}
