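# Niagara system for a blue, low-opacity diamond bling glow and sparkle effect, named for display on a player's back.
# Corona merges three burst layers, using spawn-group switches and visibility tags to drive two sprite renderers and one mesh renderer.
# A pulsing light and shield-decal sprite accompany a faint anamorphic streak that never expires by lifetime.
# Rate-spawned rising motes and two GPU sparkle emitters, with brief overbright flashes, add continuous shimmer.
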NiagaraSystem /Game/Effects/Fort_Effects/Effects/Bling/NS_Bling_Diamond_OnPlayerBack_LowOpacity
dependencies {
 usedDataInterfaces [&NiagaraDataInterfaceCurve, &NiagaraDataInterfaceVectorCurve]
 usedDynamicInputs [&ConvertVectorToPosition]
 usedModules [&SystemState]
}
emitters {
 Anamorphic {
  emitterUpdateScript [
   &EmitterState(Inactive Response=::NewEnumerator1, Life Cycle Mode=::NewEnumerator1, Loop Duration=1, UseLoopDelay=true)
   &SpawnBurst_Instantaneous(Use Spawn Probability=false, UseLoopCountLimit=true)
  ]
  particleSpawnScript [&InitializeParticle(
    Color=<rgba 0.3022 1.134 2 1>
    Lifetime=2
    Sprite Size=&RandomRangeVector2D(Minimum=<250 25>, Maximum=<400 25>, Evaluation Type=::NewEnumerator1)
    Sprite Size Mode=::NewEnumerator4
    UsePositionOffset=false
    Write Material Random=true
   )]
  particleUpdateScript [
   &ParticleState(Kill Particles When Lifetime Has Elapsed=false)
   &ScaleColor(Scale Alpha=0.06, ScaleA=true, ScaleRGB=true)
   &CameraOffset(Camera Offset Amount=50)
   &DynamicMaterialParameters(
    Index 0 Param 1=300
    Index 0 Param 2=1
    Index 0 Param 3=1
    Index 0 Param 4=1
    Module.Index0Float0Write=true
    Module.Index0Float1Write=true
    Module.Index0Float2Write=true
    Module.Index0Float3Write=true
   )
  ]
  rendererClasses [&NiagaraSpriteRendererProperties]
  renderers [{rendererClass=&NiagaraSpriteRendererProperties}]
 }
 Corona {
  emitterUpdateScript [
   &EmitterState(Inactive Response=::NewEnumerator1, Life Cycle Mode=::NewEnumerator1, Loop Duration=1.4, UseLoopDelay=true)
   &SpawnBurst_Instantaneous(Use Spawn Probability=false, UseLoopCountLimit=false)
   &SpawnBurst_Instantaneous(Spawn Group=1, Use Spawn Probability=false, UseLoopCountLimit=false) {module=SpawnBurst_Instantaneous001}
   &SpawnBurst_Instantaneous(Spawn Group=2, Use Spawn Probability=false, UseLoopCountLimit=false) {module=SpawnBurst_Instantaneous002}
  ]
  particleSpawnScript [
   /SetVariables(Particles.SpriteRotationRate=&RandomRangeFloat(Minimum=-0.1, Maximum=0.1, Evaluation Type=::NewEnumerator1)) {isSetParameters=true}
   &InitializeParticle(
    Color=&MergeHelper_ColorSwitch(
     0=<rgba 1.5312 2.3986 3 0.15>
     1=<rgba 0.0769 0.371 1 0.15>
     2=<rgba 0.5064 0.6369 1 0.0025>
     3=<rgba 0 0 0 0>
     4=<rgba 0 0 0 0>
     Index=&SpawnGroup()
     Num=::NewEnumerator3
    )
    Lifetime=&MergeHelper_FloatSwitch(
     0=&RandomRangeFloat(Minimum=1.8, Maximum=2, Evaluation Type=::NewEnumerator1)
     1=2
     2=1.9
     3=0
     4=0
     Index=&SpawnGroup()
     Num=::NewEnumerator3
    )
    Mesh Scale=&MergeHelper_VectorSwitch(0=<1 1 1>, 1=<1 1 1>, 2=<3 3 3>, 3=<0 0 0>, 4=<0 0 0>, Index=&SpawnGroup(), Num=::NewEnumerator3)
    Mesh Scale Mode=::NewEnumerator4
    Position Offset=&MergeHelper_VectorSwitch(0=<0 0 0>, 1=<0 -25 0>, 2=<0 0 0>, 3=<0 0 0>, 4=<0 0 0>, Index=&SpawnGroup(), Num=::NewEnumerator3)
    Sprite Rotation Angle=&MergeHelper_FloatSwitch(
     0=&RandomRangeFloat(Minimum=0, Maximum=1, Evaluation Type=::NewEnumerator1)
     1=&RandomRangeFloat(Minimum=0, Maximum=1, Evaluation Type=::NewEnumerator1)
     2=0
     3=0
     4=0
     Index=&SpawnGroup()
     Num=::NewEnumerator3
    )
    Sprite Rotation Mode=::NewEnumerator1
    Sprite Size=&MergeHelper_Vector2Switch(
     0=&Vector2DFromFloat(Value=&RandomRangeFloat(Minimum=100, Maximum=120, Evaluation Type=::NewEnumerator1))
     1=&Vector2DFromFloat(Value=&RandomRangeFloat(Minimum=300, Maximum=400, Evaluation Type=::NewEnumerator1))
     2=<0 0>
     3=<0 0>
     4=<0 0>
     Index=&SpawnGroup()
     Num=::NewEnumerator3
    )
    Sprite Size Mode=::NewEnumerator4
    UsePositionOffset=true
    Write Material Random=true
   )
   /SetVariables(Particles.VisibilityTag=&MergeHelper_IntSwitch(0=0, 1=1, 2=2, 3=3, 4=4, 5=?, Index=$Emitter.SpawnGroup, Num=::NewEnumerator3)) {isSetParameters=true}
  ]
  particleUpdateScript [
   &ParticleState()
   &SpriteRotationRate(Rotation Rate=&MergeHelper_FloatSwitch(
     0=&AngleConversion(
      Angle:NiagaraFloat=$Particles.SpriteRotationRate
      Angle Input:ENiagara_AngleInput=::NewEnumerator1
      Angle Output:ENiagara_AngleInput=::NewEnumerator0
     )
     1=0
     2=0
     3=0
     4=0
     Index=&SpawnGroup()
     Num=::NewEnumerator3
    ))
   &ScaleSpriteSize(
    Scale Sprite Size Mode=::NewEnumerator0
    Uniform Scale Factor=&MergeHelper_FloatSwitch(
     0=&FloatFromCurve(FloatCurve=DataInterface() {propertyValues {Curve {keys [<at 0 1>,<at 0.5 0.7>,<at 1 1>]}}})
     1=&FloatFromCurve(FloatCurve=DataInterface() {propertyValues {Curve {keys [
          {arriveTangent=-0.1064, interpMode=RCIM_Cubic, time=0.0019, value=0.5035}
          {arriveTangent=-0.007, interpMode=RCIM_Cubic, leaveTangent=-0.007, time=0.5, value=1}
          <at 1.0038 0.4965 RCIM_Cubic>
         ]}}})
     2=1
     3=0
     4=0
     Index=&SpawnGroup()
     Num=::NewEnumerator3
    )
   )
   &CameraOffset(Camera Offset Amount=&MergeHelper_FloatSwitch(0=25, 1=15, 2=25, 3=0, 4=0, Index=&SpawnGroup(), Num=::NewEnumerator3))
   &DynamicMaterialParameters(
    Index 0 Param 1=&MergeHelper_FloatSwitch(
     0=&FloatFromCurve(FloatCurve=DataInterface() {propertyValues {Curve {keys [<at 0 1 RCIM_Cubic>,<at 0.5 0.5 RCIM_Cubic>,<at 1 1>]}}})
     1=1
     2=&FloatFromCurve(FloatCurve=DataInterface() {propertyValues {Curve {keys [<at 0 2.5>,<at 1 2>]}}})
     3=0
     4=0
     Index=&SpawnGroup()
     Num=::NewEnumerator3
    )
    Index 0 Param 2=&MergeHelper_FloatSwitch(
     0=&RandomRangeFloat(Minimum=0, Maximum=0.15)
     1=0
     2=2
     3=0
     4=0
     Index=&SpawnGroup()
     Num=::NewEnumerator3
    )
    Index 0 Param 3=&MergeHelper_FloatSwitch(
     0=&RandomRangeFloat(Minimum=0.7, Maximum=0.8)
     1=1
     2=15
     3=0
     4=0
     Index=&SpawnGroup()
     Num=::NewEnumerator3
    )
    Index 0 Param 4=&MergeHelper_FloatSwitch(0=1, 1=1, 2=1, 3=0, 4=0, Index=&SpawnGroup(), Num=::NewEnumerator3)
    Module.Index0Float0Write=true
    Module.Index0Float1Write=true
    Module.Index0Float2Write=true
    Module.Index0Float3Write=true
   )
   &ScaleColor(
    Scale Alpha=&MergeHelper_FloatSwitch(
     0=&FloatFromCurve(FloatCurve=DataInterface() {propertyValues {Curve {keys [
          <at 0 0 RCIM_Cubic>
          {arriveTangent=2.5, interpMode=RCIM_Cubic, leaveTangent=2.5, time=0.1631, value=0.6575}
          {arriveTangent=0.1813, interpMode=RCIM_Cubic, leaveTangent=0.1813, time=0.4, value=1}
          {arriveTangent=-1.6667, interpMode=RCIM_Cubic, leaveTangent=-1.6667, time=0.7223, value=0.7588}
          <at 1 0>
         ]}}})
     1=&FloatFromCurve(FloatCurve=DataInterface() {propertyValues {Curve {keys [
          <at 0.0011 0.0003 RCIM_Cubic>
          {arriveTangent=-0.0003, interpMode=RCIM_Cubic, leaveTangent=-0.0003, time=0.4951, value=0.7894}
          {interpMode=RCIM_Cubic, leaveTangent=-1.3318, time=1, value=0}
         ]}}})
     2=1
     3=0
     4=0
     Index=&SpawnGroup()
     Num=::NewEnumerator3
    )
    Scale RGB=&MergeHelper_VectorSwitch(
     0=&VectorFromCurve(VectorCurve=DataInterface() {propertyValues {
        XCurve {keys [<at 0 1 RCIM_Cubic>,<at 1 0.3728>]}
        YCurve {keys [<at 0 0.6851 RCIM_Cubic>,<at 1 1>]}
        ZCurve {keys [<at 0 0.8647 RCIM_Cubic>,<at 1 0.6644>]}
       }})
     1=&VectorFromFloat(Value=&FloatFromCurve(FloatCurve=DataInterface() {propertyValues {Curve {keys [<at 0 10 RCIM_Cubic>,<at 0.05 1>]}}}))
     2=&VectorFromCurve(VectorCurve=DataInterface() {propertyValues {
        XCurve {keys [<at 0 0 RCIM_Cubic>,<at 0.4995 0.0001 RCIM_Cubic>,<at 1 0 RCIM_Cubic>]}
        YCurve {keys [<at 0 0 RCIM_Cubic>,<at 0.4995 0.5562 RCIM_Cubic>,<at 1 0 RCIM_Cubic>]}
        ZCurve {keys [<at 0 0 RCIM_Cubic>,<at 0.4995 1 RCIM_Cubic>,<at 1 0 RCIM_Cubic>]}
       }})
     3=<0 0 0>
     4=<0 0 0>
     Index=&SpawnGroup()
     Num=::NewEnumerator3
    )
    ScaleA=true
    ScaleRGB=true
   )
  ]
  rendererClasses [&NiagaraSpriteRendererProperties, &NiagaraMeshRendererProperties]
  renderers [
   {rendererClass=&NiagaraSpriteRendererProperties}
   {rendererClass=&NiagaraSpriteRendererProperties, rendererIndex=1}
   {rendererClass=&NiagaraMeshRendererProperties, rendererIndex=2}
  ]
 }
 Light {
  emitterUpdateScript [
   &EmitterState(Life Cycle Mode=::NewEnumerator1, Loop Duration=2, UseLoopDelay=true)
   &SpawnBurst_Instantaneous(Use Spawn Probability=false, UseLoopCountLimit=false)
  ]
  particleSpawnScript [
   &InitializeParticle(
    Color=<rgba 0.0769 0.371 1 0.35>
    Lifetime=2
    Position Offset=<25 -50 50>
    Sprite Rotation Angle=&RandomRangeFloat(Minimum=0, Maximum=1, Evaluation Type=::NewEnumerator1)
    Sprite Rotation Mode=::NewEnumerator1
    Sprite Size=&Vector2DFromFloat(Value=&RandomRangeFloat(Minimum=300, Maximum=400, Evaluation Type=::NewEnumerator1))
    Sprite Size Mode=::NewEnumerator4
    UsePositionOffset=true
    Write Material Random=true
   )
   &Color(Color=&MakeLinearColorFromVectorAndFloat("Vector (RGB)"=<0.0769 0.371 1>, "Float (Alpha)"=&ClampFloat(Float=0.35))) {enabled=false}
  ]
  particleUpdateScript [
   /SetVariables(Particles.LightExponent=8) {isSetParameters=true}
   &ParticleState()
   &CameraOffset(Camera Offset Amount=25)
   &ScaleColor(
    Scale Alpha=&FloatFromCurve(FloatCurve=DataInterface() {propertyValues {Curve {keys [
         <at 0.0011 0.0003 RCIM_Cubic>
         {arriveTangent=-0.0003, interpMode=RCIM_Cubic, leaveTangent=-0.0003, time=0.4951, value=0.7894}
         {interpMode=RCIM_Cubic, leaveTangent=-1.3318, time=1, value=0}
        ]}}})
    Scale RGB=&VectorFromFloat(Value=&FloatFromCurve(FloatCurve=DataInterface() {propertyValues {Curve {keys [<at 0 10 RCIM_Cubic>,<at 0.05 1>]}}}))
    ScaleA=true
    ScaleRGB=true
   )
   &CascadeConversion_LightProperties(LightBrightnessScale=32, LightColorScale=<1 1 1>, LightRadiusScale=1)
  ]
  rendererClasses [&NiagaraLightRendererProperties]
  renderers [{rendererClass=&NiagaraLightRendererProperties, rendererIndex=0}]
 }
 Particle_Emitter {
  emitterUpdateScript [
   &EmitterState(Life Cycle Mode=::NewEnumerator1, Loop Duration=1, UseLoopDelay=true)
   &SpawnRate(SpawnRate=5, Use Spawn Probability=false)
  ]
  particleSpawnScript [
   &InitializeParticle(
    Color=&MultiplyLinearColor(
     Linear Color=&RandomRangeLinearColor(Minimum=<rgba 0.8728 0.9258 1 1>, Maximum=<rgba 0.1699 0.1242 1 1>)
     Scale Factor=/Niagara/DynamicInputs/UniformRange/V2/RandomRangeVector4(
      Minimum:Vector4f=<1 1 1 0.05>
      Maximum:Vector4f=<1 1 1 0.075>
      Evaluation Type:ENiagaraRandomnessEvaluation=::NewEnumerator0
      Recalculate Random Each Loop:NiagaraBool=true
      Randomness Mode:ENiagaraRandomnessMode=::NewEnumerator0
      Random Seed:NiagaraInt32=0
      Fixed Random Seed:NiagaraBool=false
     )
    )
    Lifetime=&RandomRangeFloat(Minimum=0.6, Maximum=1.25, Evaluation Type=::NewEnumerator1)
    Position Offset=<0 -10 0>
    Sprite Size=&Vector2DFromFloat(Value=&RandomRangeFloat(Minimum=25, Maximum=50, Evaluation Type=::NewEnumerator1))
    Sprite Size Mode=::NewEnumerator4
    UsePositionOffset=true
    Write Material Random=true
   )
   &Color(Color=&MakeLinearColorFromVectorAndFloat(
     "Vector (RGB)"=&RandomRangeVector(Minimum=<0.8728 0.9258 1>, Maximum=<0.1699 0.1242 1>, Evaluation Type=::NewEnumerator1)
     "Float (Alpha)"=&ClampFloat(Float=&RandomRangeFloat(Minimum=0.05, Maximum=0.075, Evaluation Type=::NewEnumerator1))
    )) {enabled=false}
   &ShapeLocation(MaskSpawnGroup=false, Override Seed=false, Rotation Axis=<0 0 1>, Rotation Mode=::NewEnumerator1, Sphere Radius=15, UseOwnerScale=true)
  ]
  particleUpdateScript [
   &ParticleState()
   &ScaleSpriteSize(Scale Sprite Size Mode=::NewEnumerator0, Uniform Scale Factor=0.5)
   &CameraOffset(Camera Offset Amount=35)
   &ScaleColor(
    Scale Alpha=&FloatFromCurve(
     FloatCurve=DataInterface() {propertyValues {Curve {keys [<at 0 0 RCIM_Cubic>,<at 0.4429 0.9757 RCIM_Cubic>,<at 1 0 RCIM_Cubic>]}}}
     Scale Curve=1.3
    )
    Scale RGB=&VectorFromCurve(VectorCurve=DataInterface() {propertyValues {XCurve {keys [<at 0 1>,<at 1 1>]}, YCurve {keys [<at 0 1>,<at 1 1>]}, ZCurve {keys [<at 0 1>,<at 1 1>]}}})
    ScaleA=true
    ScaleRGB=true
   )
   &AccelerationForce(Acceleration=<0 0 25>)
   &SolveForcesAndVelocity(Clamp Velocity=false, Limit Acceleration=false)
   &DynamicMaterialParameters(
    Index 0 Param 1=1
    Index 0 Param 2=1
    Index 0 Param 3=1
    Index 0 Param 4=1
    Module.Index0Float0Write=true
    Module.Index0Float1Write=true
    Module.Index0Float2Write=true
    Module.Index0Float3Write=true
   )
  ]
  rendererClasses [&NiagaraSpriteRendererProperties]
  renderers [{rendererClass=&NiagaraSpriteRendererProperties}]
 }
 Particle_Emitter001 {
  simTarget=GPUComputeSim
  emitterUpdateScript [
   &EmitterState(Life Cycle Mode=::NewEnumerator1, Loop Duration=1, UseLoopDelay=true)
   &SpawnRate(SpawnRate=15, Use Spawn Probability=false)
  ]
  particleSpawnScript [
   &InitializeParticle(
    Color=<rgba 210.1058 500 414.0759 0.05>
    Lifetime=&RandomRangeFloat(Minimum=0.5, Maximum=1, Evaluation Type=::NewEnumerator1)
    Position Offset=&RandomRangeVector(Minimum=<-25 -15 -15>, Maximum=<15 0 15>, Evaluation Type=::NewEnumerator1)
    Sprite Rotation Angle=&RandomRangeFloat(Minimum=-0.02, Maximum=0.02, Evaluation Type=::NewEnumerator1)
    Sprite Rotation Mode=::NewEnumerator1
    Sprite Size=&Vector2DFromFloat(Value=&RandomRangeFloat(Minimum=6, Maximum=10, Evaluation Type=::NewEnumerator1))
    Sprite Size Mode=::NewEnumerator4
    UsePositionOffset=true
    Write Material Random=true
   )
   &Color(Color=&MakeLinearColorFromVectorAndFloat("Vector (RGB)"=<210.1058 500 414.0759>, "Float (Alpha)"=&ClampFloat(Float=0.05))) {enabled=false}
   &ShapeLocation(MaskSpawnGroup=false, Override Seed=false, Sphere Radius=2, UseOwnerScale=true)
  ]
  particleUpdateScript [
   &ParticleState()
   &ScaleSpriteSize(
    Scale Sprite Size Mode=::NewEnumerator0
    Uniform Scale Factor=&FloatFromCurve(FloatCurve=DataInterface() {propertyValues {Curve {keys [{arriveTangent=-0.421, interpMode=RCIM_Cubic, time=0, value=0}, <at 0.3594 0.8427 RCIM_Cubic>, <at 1 0 RCIM_Cubic>]}}})
   )
   &ScaleColor(
    Scale Alpha=&FloatFromCurve(FloatCurve=DataInterface() {propertyValues {Curve {keys [
         <at 0 0.0994 RCIM_Cubic>
         {arriveTangent=-0.0994, interpMode=RCIM_Cubic, leaveTangent=-0.0994, time=0.128, value=0.0373}
         <at 1 0 RCIM_Cubic>
        ]}}})
    Scale RGB=&VectorFromFloat(Value=&FloatFromCurve(FloatCurve=DataInterface() {propertyValues {Curve {keys [{arriveTangent=-14.4521, interpMode=RCIM_Cubic, time=0, value=15}, <at 0.4999 0 RCIM_Cubic>]}}}))
    ScaleA=true
    ScaleRGB=true
   )
  ]
  rendererClasses [&NiagaraSpriteRendererProperties]
  renderers [{rendererClass=&NiagaraSpriteRendererProperties}]
 }
 Pot_of_Gold_Shimmer {
  simTarget=GPUComputeSim
  emitterUpdateScript [
   &EmitterState(Life Cycle Mode=::NewEnumerator1, Loop Duration=1, UseLoopDelay=true)
   &SpawnRate(SpawnRate=8, Use Spawn Probability=false)
  ]
  particleSpawnScript [
   &InitializeParticle(
    Color=&MultiplyLinearColor(
     Linear Color=<rgba 4.974 15.7608 20 1>
     Scale Factor=&RandomRangeVector4(
      Minimum:Vector4f=<1 1 1 0.25>
      Maximum:Vector4f=<1 1 1 0.5>
      Evaluation Type:ENiagaraRandomnessEvaluation=::NewEnumerator0
      Recalculate Random Each Loop:NiagaraBool=true
      Randomness Mode:ENiagaraRandomnessMode=::NewEnumerator0
      Random Seed:NiagaraInt32=0
      Fixed Random Seed:NiagaraBool=false
     )
    )
    Lifetime=&RandomRangeFloat(Minimum=0.25, Maximum=0.5, Evaluation Type=::NewEnumerator1)
    Position Offset=&RandomRangeVector(Minimum=<-25 -15 -15>, Maximum=<15 0 15>, Evaluation Type=::NewEnumerator1)
    Sprite Rotation Angle=&RandomRangeFloat(Minimum=0.02, Maximum=-0.02, Evaluation Type=::NewEnumerator1)
    Sprite Rotation Mode=::NewEnumerator1
    Sprite Size=&Vector2DFromFloat(Value=&RandomRangeFloat(Minimum=2, Maximum=4, Evaluation Type=::NewEnumerator1))
    Sprite Size Mode=::NewEnumerator4
    UsePositionOffset=true
    Write Material Random=true
   )
   &Color(Color=&MakeLinearColorFromVectorAndFloat(
     "Vector (RGB)"=<4.974 15.7608 20>
     "Float (Alpha)"=&ClampFloat(Float=&RandomRangeFloat(Minimum=0.25, Maximum=0.5, Evaluation Type=::NewEnumerator1))
    )) {enabled=false}
   &ShapeLocation(MaskSpawnGroup=false, Override Seed=false, Sphere Radius=2, UseOwnerScale=true)
  ]
  particleUpdateScript [
   &ParticleState()
   &ScaleColor(
    Scale Alpha=&FloatFromCurve(FloatCurve=DataInterface() {propertyValues {Curve {keys [<at 0 1>,<at 1 0>]}}})
    Scale RGB=&VectorFromCurve(VectorCurve=DataInterface() {propertyValues {XCurve {keys [<at 0 50>,<at 0.1 1>]}, YCurve {keys [<at 0 50>,<at 0.1 1>]}, ZCurve {keys [<at 0 50>,<at 0.1 1>]}}})
    ScaleA=true
    ScaleRGB=true
   )
  ]
  rendererClasses [&NiagaraSpriteRendererProperties]
  renderers [{rendererClass=&NiagaraSpriteRendererProperties}]
 }
 ShieldDecals {
  emitterUpdateScript [
   &EmitterState(Life Cycle Mode=::NewEnumerator1, Loop Duration=1, UseLoopDelay=true)
   &SpawnBurst_Instantaneous(Use Spawn Probability=false, UseLoopCountLimit=false)
  ]
  particleSpawnScript [
   &InitializeParticle(
    Color=&RandomRangeLinearColor(Minimum=<rgba 0.1669 0.34 1 1>, Maximum=<rgba 1.5511 2.8011 5 1>)
    Lifetime=2
    Position Offset=<0 -25 0>
    Sprite Rotation Angle=&RandomRangeFloat(Minimum=0, Maximum=1, Evaluation Type=::NewEnumerator1)
    Sprite Rotation Mode=::NewEnumerator1
    Sprite Size=<60 60>
    Sprite Size Mode=::NewEnumerator4
    UsePositionOffset=true
    Write Material Random=true
   )
   &Color(Color=&MakeLinearColorFromVectorAndFloat(
     "Vector (RGB)"=&RandomRangeVector(Minimum=<0.1669 0.34 1>, Maximum=<1.5511 2.8011 5>, Evaluation Type=::NewEnumerator1)
     "Float (Alpha)"=&ClampFloat(Float=1)
    )) {enabled=false}
  ]
  particleUpdateScript [
   &ParticleState()
   &ScaleColor(
    Scale Alpha=&FloatFromCurve(FloatCurve=DataInterface() {propertyValues {Curve {keys [{arriveTangent=-0.1447, interpMode=RCIM_Cubic, time=0, value=0}, <at 0.4739 0.9561 RCIM_Cubic>, <at 1 0>]}}})
    Scale RGB=&VectorFromCurve(VectorCurve=DataInterface() {propertyValues {
       XCurve {keys [<at 0 0>,<at 0.5 0.1>,<at 1 0>]}
       YCurve {keys [<at 0 0>,<at 0.5 0.25>,<at 1 0>]}
       ZCurve {keys [<at 0 0>,<at 0.5 5>,<at 1 0>]}
      }})
    ScaleA=true
    ScaleRGB=true
   )
   &CameraOffset(Camera Offset Amount=5)
   &DynamicMaterialParameters(
    Index 0 Param 1=&FloatFromCurve(FloatCurve=DataInterface() {propertyValues {Curve {keys [
         <at 0 -0.5 RCIM_Cubic>
         {arriveTangent=10, interpMode=RCIM_Cubic, leaveTangent=10, time=0.0692, value=0.2738}
         {arriveTangent=5.0777, interpMode=RCIM_Cubic, leaveTangent=5.0777, time=0.15, value=1}
         {interpMode=RCIM_Cubic, leaveTangent=4.2839, time=1, value=5}
        ]}}})
    Index 0 Param 2=1
    Index 0 Param 3=&FloatFromCurve(FloatCurve=DataInterface() {propertyValues {Curve {keys [{arriveTangent=0.1967, interpMode=RCIM_Cubic, time=0, value=0.8609}, {interpMode=RCIM_Cubic, leaveTangent=0.1844, time=1, value=1}]}}})
    Index 0 Param 4=1
    Module.Index0Float0Write=true
    Module.Index0Float1Write=true
    Module.Index0Float2Write=true
    Module.Index0Float3Write=true
   )
  ]
  rendererClasses [&NiagaraSpriteRendererProperties]
  renderers [{rendererClass=&NiagaraSpriteRendererProperties}]
 }
}
summary {
 systemName=NS_Bling_Diamond_OnPlayerBack_LowOpacity
 emitters [
  {emitterName=Corona, rendererClasses [&NiagaraSpriteRendererProperties, &NiagaraMeshRendererProperties]}
  {emitterName=Particle_Emitter, rendererClasses [&NiagaraSpriteRendererProperties]}
  {emitterName=Light, rendererClasses [&NiagaraLightRendererProperties]}
  {emitterName=Particle_Emitter001, simTarget=GPUComputeSim, rendererClasses [&NiagaraSpriteRendererProperties]}
  {emitterName=Pot_of_Gold_Shimmer, simTarget=GPUComputeSim, rendererClasses [&NiagaraSpriteRendererProperties]}
  {emitterName=Anamorphic, rendererClasses [&NiagaraSpriteRendererProperties]}
  {emitterName=ShieldDecals, rendererClasses [&NiagaraSpriteRendererProperties]}
 ]
 userVariables []
}
