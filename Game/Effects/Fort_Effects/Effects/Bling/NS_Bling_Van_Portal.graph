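# Niagara one-shot-style portal burst for the "Bling Van" effect, mostly blue-cyan, built from eleven independent emitters with no user parameters.
# Mesh emitters draw cracks, floating fragments, electric arcs, a shockwave, glow and a comet; sprites add corona and flares, plus a flickering light.
# Each emitter burst-spawns, with arcs also spawning at a rate.
# Lifetime curves drive scale, HDR color and dynamic material parameters; camera offsets layer the visible elements.
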NiagaraSystem /Game/Effects/Fort_Effects/Effects/Bling/NS_Bling_Van_Portal
dependencies {
 usedDataInterfaces [
  &NiagaraDataInterfaceCurve
  &NiagaraDataInterfaceVectorCurve
  &NiagaraDataInterfaceVector2DCurve
 ]
 usedDynamicInputs [&ConvertVectorToPosition]
 usedModules [&SystemState]
}
emitters {
 Comet {
  emitterUpdateScript [
   &EmitterState(Inactive Response=::NewEnumerator1, Life Cycle Mode=::NewEnumerator1, Loop Behavior=::NewEnumerator1, Loop Duration=0.2, UseLoopDelay=true)
   &SpawnBurst_Instantaneous(Use Spawn Probability=false, UseLoopCountLimit=false)
  ]
  particleSpawnScript [
   &InitializeParticle(
    Color=<rgba 0.15 0.6 1 1>
    Lifetime=0.5
    Mesh Scale=<8 8 8>
    Mesh Scale Mode=::NewEnumerator4
    Position Offset=<50 0 0>
    UsePositionOffset=true
    Write Material Random=true
   )
   &AddVelocity(Velocity=<10 0 0>)
   &Color(Color=&MakeLinearColorFromVectorAndFloat("Vector (RGB)"=<0.15 0.6 1>, "Float (Alpha)"=&ClampFloat(Float=1))) {enabled=false}
  ]
  particleUpdateScript [
   &ParticleState()
   &ScaleMeshSize(Scale Factor=&VectorFromCurve(VectorCurve=DataInterface() {propertyValues {XCurve {keys [<at 0 0.6>,<at 1 1>]}, YCurve {keys [<at 0 0.6>,<at 1 1>]}, ZCurve {keys [<at 0 0.6>,<at 1 1>]}}}))
   &SolveForcesAndVelocity(Clamp Velocity=false, Limit Acceleration=false)
   &ScaleColor(
    Scale Alpha=&FloatFromCurve(FloatCurve=DataInterface() {propertyValues {Curve {keys [<at 0 0 RCIM_Cubic>,<at 0.0974 0.9951 RCIM_Cubic>,<at 1 0>]}}}, Scale Curve=0.7)
    Scale RGB=&VectorFromCurve(VectorCurve=DataInterface() {propertyValues {XCurve {keys [<at 0 2>,<at 1 0.985>]}, YCurve {keys [<at 0 2>,<at 1 0.6599>]}, ZCurve {keys [<at 0 2>,<at 1 1>]}}})
    ScaleA=true
    ScaleRGB=true
   )
   &CameraOffset(Camera Offset Amount=2)
  ]
  rendererClasses [&NiagaraMeshRendererProperties]
  renderers [{rendererClass=&NiagaraMeshRendererProperties, rendererIndex=0}]
 }
 Corona {
  emitterUpdateScript [
   &EmitterState(Life Cycle Mode=::NewEnumerator1, Loop Behavior=::NewEnumerator1, Loop Duration=1, UseLoopDelay=true)
   &SpawnBurst_Instantaneous(Use Spawn Probability=false, UseLoopCountLimit=false)
  ]
  particleSpawnScript [
   &InitializeParticle(
    Color=<rgba 1.1458 2.1149 4 1>
    Lifetime=0.3
    Sprite Size=<1400 1400>
    Sprite Size Mode=::NewEnumerator4
    UsePositionOffset=false
    Write Material Random=true
   )
   &Color(Color=&MakeLinearColorFromVectorAndFloat("Vector (RGB)"=<1.1458 2.1149 4>, "Float (Alpha)"=&ClampFloat(Float=1))) {enabled=false}
  ]
  particleUpdateScript [
   &ParticleState()
   &ScaleSpriteSize(
    Scale Sprite Size Mode=::NewEnumerator0
    Uniform Scale Factor=&FloatFromCurve(FloatCurve=DataInterface() {propertyValues {Curve {keys [{arriveTangent=1.4067, interpMode=RCIM_Cubic, time=0, value=0}, {interpMode=RCIM_Cubic, leaveTangent=0.2191, time=1, value=1}]}}})
   )
   &ScaleColor(
    Scale Alpha=&FloatFromCurve(FloatCurve=DataInterface() {propertyValues {Curve {keys [<at 0 1 RCIM_Cubic>,<at 1 0>]}}})
    Scale RGB=<2 2 2>
    ScaleA=true
    ScaleRGB=true
   )
   &DynamicMaterialParameters(
    Index 0 Param 1=&FloatFromCurve(FloatCurve=DataInterface() {propertyValues {Curve {keys [{arriveTangent=-1.5232, interpMode=RCIM_Cubic, time=0, value=1.5}, <at 1 0.7 RCIM_Cubic>]}}})
    Index 0 Param 2=0.4
    Index 0 Param 3=0.6
    Index 0 Param 4=1
    Module.Index0Float0Write=true
    Module.Index0Float1Write=true
    Module.Index0Float2Write=true
    Module.Index0Float3Write=true
   )
   &CameraOffset(Camera Offset Amount=400)
  ]
  rendererClasses [&NiagaraSpriteRendererProperties]
  renderers [{rendererClass=&NiagaraSpriteRendererProperties}]
 }
 CrackSpline {
  emitterUpdateScript [
   &EmitterState(Inactive Response=::NewEnumerator1, Life Cycle Mode=::NewEnumerator1, Loop Behavior=::NewEnumerator1, Loop Duration=2, UseLoopDelay=true)
   &SpawnBurst_Instantaneous(Spawn Count=3, Use Spawn Probability=false, UseLoopCountLimit=false)
  ]
  particleSpawnScript [
   &InitializeParticle(
    Lifetime=&RandomRangeFloat(Minimum=1.25, Maximum=1.32, Evaluation Type=::NewEnumerator1)
    Mesh Scale=&VectorFromFloat(Value=&RandomRangeFloat(Minimum=11, Maximum=12, Evaluation Type=::NewEnumerator1))
    Mesh Scale Mode=::NewEnumerator4
    UsePositionOffset=false
    Write Material Random=true
   )
   &Color(Color=&MakeLinearColorFromVectorAndFloat("Vector (RGB)"=<1 1 1>, "Float (Alpha)"=&ClampFloat(Float=1))) {enabled=false}
   /Niagara/Modules/CascadeConversion/CascadeConversion_InitialMeshRotation(Rotation:Vector3f=&RandomRangeVector(Minimum=<0 0 0>, Maximum=<1 1 1>, Evaluation Type=::NewEnumerator1))
  ]
  particleUpdateScript [
   &ParticleState()
   &ScaleMeshSize(Scale Factor=&VectorFromFloat(Value=&FloatFromCurve(FloatCurve=DataInterface() {propertyValues {Curve {keys [
          <at 0 0.9029 RCIM_Cubic>
          {arriveTangent=1.0111, interpMode=RCIM_Cubic, leaveTangent=1.0111, time=0.0314, value=1.0157}
          {arriveTangent=-0.2487, interpMode=RCIM_Cubic, leaveTangent=-0.2487, time=0.1139, value=1.0181}
          {arriveTangent=-0.2461, interpMode=RCIM_Cubic, leaveTangent=-0.2461, time=0.1545, value=0.9851}
          {interpMode=RCIM_Cubic, leaveTangent=-0.3532, time=1, value=0.8}
         ]}}})))
   &ScaleColor(
    Scale Alpha=&FloatFromCurve(FloatCurve=DataInterface() {propertyValues {Curve {keys [<at 0.77 1 RCIM_Cubic>,<at 1 0>]}}})
    Scale RGB=&VectorFromCurve(VectorCurve=DataInterface() {propertyValues {
       XCurve {keys [
         {arriveTangent=-6.9225, interpMode=RCIM_Cubic, time=0, value=4.5}
         {arriveTangent=-3.6158, interpMode=RCIM_Cubic, leaveTangent=-3.6158, time=0.65, value=1}
         <at 1 0.8842 RCIM_Cubic>
        ]}
       YCurve {keys [
         {arriveTangent=-15.8226, interpMode=RCIM_Cubic, time=0, value=6}
         {arriveTangent=-5.6986, interpMode=RCIM_Cubic, leaveTangent=-5.6986, time=0.65, value=1}
         {interpMode=RCIM_Cubic, leaveTangent=-5.843, time=1, value=0.3014}
        ]}
       ZCurve {keys [
         {arriveTangent=-15.2224, interpMode=RCIM_Cubic, time=0, value=6}
         {arriveTangent=-5, interpMode=RCIM_Cubic, leaveTangent=-5, time=0.65, value=1}
         <at 1 1 RCIM_Cubic>
        ]}
      }})
    ScaleA=true
    ScaleRGB=true
   )
   &DynamicMaterialParameters(
    Index 0 Param 1=&FloatFromCurve(FloatCurve=DataInterface() {propertyValues {Curve {keys [
         <at 0 1 RCIM_Cubic>
         {arriveTangent=-1, interpMode=RCIM_Cubic, leaveTangent=-1, time=0.3329, value=0.616}
         {interpMode=RCIM_Cubic, leaveTangent=-0.3905, time=1, value=0}
        ]}}})
    Index 0 Param 2=&FloatFromCurve(FloatCurve=DataInterface() {propertyValues {Curve {keys [
         {arriveTangent=-1.9782, interpMode=RCIM_Cubic, time=0, value=1}
         {arriveTangent=-1, interpMode=RCIM_Cubic, leaveTangent=-1, time=0.65, value=0.5}
         <at 1 0 RCIM_Cubic>
        ]}}})
    Index 0 Param 3=&FloatFromCurve(FloatCurve=DataInterface() {propertyValues {Curve {keys [{arriveTangent=1.0145, interpMode=RCIM_Cubic, time=0, value=-0.1}, <at 0.25 0>]}}})
    Index 0 Param 4=&FloatFromCurve(FloatCurve=DataInterface() {propertyValues {Curve {keys [<at 0 0 RCIM_Cubic>,<at 1 0>]}}})
    Module.Index0Float0Write=true
    Module.Index0Float1Write=true
    Module.Index0Float2Write=true
    Module.Index0Float3Write=true
   )
   &CameraOffset(Camera Offset Amount=200)
  ]
  rendererClasses [&NiagaraMeshRendererProperties]
  renderers [{rendererClass=&NiagaraMeshRendererProperties, rendererIndex=0}]
 }
 ElectricArcsResidual {
  emitterUpdateScript [
   &EmitterState(Life Cycle Mode=::NewEnumerator1, Loop Behavior=::NewEnumerator1, Loop Duration=0.5, UseLoopDelay=true)
   &SpawnBurst_Instantaneous(Spawn Count=6, Use Spawn Probability=false, UseLoopCountLimit=false)
   &SpawnRate(SpawnRate=15, Use Spawn Probability=false)
  ]
  particleSpawnScript [
   &InitializeParticle(
    Color=<rgba 0.0986 1 0.6632 2>
    Lifetime=&RandomRangeFloat(Minimum=0.08, Maximum=0.2, Evaluation Type=::NewEnumerator1)
    Mesh Scale=&RandomRangeVector(Minimum=<20 6 6>, Maximum=<25 20 20>, Evaluation Type=::NewEnumerator1)
    Mesh Scale Mode=::NewEnumerator4
    UsePositionOffset=false
    Write Material Random=true
   )
   &DynamicMaterialParameters(
    Index 0 Param 1=&RandomRangeFloat(Minimum=2, Maximum=3, Evaluation Type=::NewEnumerator1)
    Index 0 Param 3=&RandomRangeFloat(Minimum=0.2, Maximum=2, Evaluation Type=::NewEnumerator1)
    Index 0 Param 4=&RandomRangeFloat(Minimum=0, Maximum=1, Evaluation Type=::NewEnumerator1)
    Module.Index0Float0Write=true
    Module.Index0Float1Write=false
    Module.Index0Float2Write=true
    Module.Index0Float3Write=true
   )
   &CascadeConversion_InitialMeshRotation(Rotation:Vector3f=&RandomRangeVector(Minimum=<0 0 0>, Maximum=<1 1 1>, Evaluation Type=::NewEnumerator1))
   &Color(Color=&MakeLinearColorFromVectorAndFloat("Vector (RGB)"=<0.0986 1 0.6632>, "Float (Alpha)"=&ClampFloat(Float=2))) {enabled=false}
   &ShapeLocation(MaskSpawnGroup=false, Override Seed=false, Sphere Radius=10, UseOwnerScale=true)
   &AddVelocity(Velocity=&Multiply_VectorByFloat(Vector=$Particles.ShapeLocation.ShapeVector, Float=5))
  ]
  particleUpdateScript [
   &ParticleState()
   &ScaleMeshSize(Scale Factor=&VectorFromCurve(VectorCurve=DataInterface() {propertyValues {
       XCurve {keys [<at 0.0033 0.5 RCIM_Constant>,<at 0.2054 0.7619 RCIM_Constant>,<at 0.4101 0.8575 RCIM_Constant>,<at 0.6967 1>]}
       YCurve {keys [<at 0.0033 0.6004 RCIM_Constant>,<at 0.2054 0.6716 RCIM_Constant>,<at 0.4101 0.3646 RCIM_Constant>,<at 0.6967 0.2468>]}
       ZCurve {keys [<at 0.0033 0.7767 RCIM_Constant>,<at 0.2054 0.5215 RCIM_Constant>,<at 0.4101 0.4457 RCIM_Constant>,<at 0.6967 0.6282>]}
      }}))
   &DynamicMaterialParameters(
    Index 0 Param 2=&FloatFromCurve(FloatCurve=DataInterface() {propertyValues {Curve {keys [{arriveTangent=-1.8894, interpMode=RCIM_Cubic, time=0, value=1}, <at 1 -0.25>]}}})
    Module.Index0Float0Write=false
    Module.Index0Float1Write=true
    Module.Index0Float2Write=false
    Module.Index0Float3Write=false
   ) {module=DynamicMaterialParameters001}
   &ScaleColor(
    Scale RGB=&VectorFromCurve(VectorCurve=DataInterface() {propertyValues {
       XCurve {keys [<at 0 50 RCIM_Cubic>, {arriveTangent=-45, interpMode=RCIM_Cubic, leaveTangent=-45, time=0.5, value=30}, <at 1 5>]}
       YCurve {keys [<at 0 19.1123 RCIM_Cubic>, {arriveTangent=-18.2463, interpMode=RCIM_Cubic, leaveTangent=-18.2463, time=0.5, value=13.6826}, <at 1 0.8661>]}
       ZCurve {keys [<at 0 22.1016 RCIM_Cubic>, {arriveTangent=-21.4843, interpMode=RCIM_Cubic, leaveTangent=-21.4843, time=0.5, value=12.787}, <at 1 0.6173>]}
      }})
    ScaleA=true
    ScaleRGB=true
   )
   &CameraOffset(Camera Offset Amount=200)
   &SolveForcesAndVelocity(Clamp Velocity=false, Limit Acceleration=false)
  ]
  rendererClasses [&NiagaraMeshRendererProperties]
  renderers [{rendererClass=&NiagaraMeshRendererProperties, rendererIndex=0}]
 }
 Flare {
  emitterUpdateScript [
   &EmitterState(Inactive Response=::NewEnumerator1, Life Cycle Mode=::NewEnumerator1, Loop Behavior=::NewEnumerator1, Loop Duration=1, UseLoopDelay=true)
   &SpawnBurst_Instantaneous(Use Spawn Probability=false, UseLoopCountLimit=false)
  ]
  particleSpawnScript [
   &InitializeParticle(
    Color=<rgba 0.0701 0.2462 1 0.25>
    Lifetime=0.25
    Sprite Rotation Angle=&RandomRangeFloat(Minimum=0, Maximum=1, Evaluation Type=::NewEnumerator1)
    Sprite Rotation Mode=::NewEnumerator1
    Sprite Size=<3000 3000>
    Sprite Size Mode=::NewEnumerator4
    UsePositionOffset=false
    Write Material Random=true
   )
   &Color(Color=&MakeLinearColorFromVectorAndFloat("Vector (RGB)"=<0.0701 0.2462 1>, "Float (Alpha)"=&ClampFloat(Float=0.25))) {enabled=false}
  ]
  particleUpdateScript [
   &ParticleState()
   &ScaleSpriteSize(
    Scale Sprite Size Mode=::NewEnumerator0
    Uniform Scale Factor=&FloatFromCurve(FloatCurve=DataInterface() {propertyValues {Curve {keys [
         <at 0 0.5 RCIM_Cubic>
         {arriveTangent=4.2512, interpMode=RCIM_Cubic, leaveTangent=4.2512, time=0.4842, value=1.2326}
         {arriveTangent=5.7247, interpMode=RCIM_Cubic, leaveTangent=5.7247, time=0.5728, value=2.9351}
         {arriveTangent=-14.371, interpMode=RCIM_Cubic, leaveTangent=-14.371, time=0.6024, value=1.9096}
         {arriveTangent=-6.9582, interpMode=RCIM_Cubic, leaveTangent=-6.9582, time=0.6189, value=2.2723}
         {arriveTangent=1.1897, interpMode=RCIM_Cubic, leaveTangent=1.1897, time=0.7009, value=1.2244}
         {arriveTangent=-3.4248, interpMode=RCIM_Cubic, leaveTangent=-3.4248, time=0.7316, value=2.4064}
         {interpMode=RCIM_Cubic, leaveTangent=-15.1861, time=1, value=0.2}
        ]}}})
   )
   &ScaleColor(
    Scale Alpha=&FloatFromCurve(
     FloatCurve=DataInterface() {propertyValues {Curve {keys [<at 0.9 1 RCIM_Cubic>, {interpMode=RCIM_Cubic, leaveTangent=-26.729, time=1, value=0}]}}}
     Scale Curve=0.7
    )
    Scale RGB=&VectorFromCurve(VectorCurve=DataInterface() {propertyValues {
       XCurve {keys [
         <at 0 0.4792 RCIM_Cubic>
         {arriveTangent=3.2896, interpMode=RCIM_Cubic, leaveTangent=3.2896, time=0.48, value=2.0312}
         {arriveTangent=1.863, interpMode=RCIM_Cubic, leaveTangent=1.863, time=0.5336, value=2.2347}
         <at 1 3>
        ]}
       YCurve {keys [
         <at 0 0.594 RCIM_Cubic>
         {arriveTangent=2.9952, interpMode=RCIM_Cubic, leaveTangent=2.9952, time=0.48, value=2.07}
         {arriveTangent=-2.0416, interpMode=RCIM_Cubic, leaveTangent=-2.0416, time=0.5336, value=2.1924}
         <at 1 1.0084>
        ]}
       ZCurve {keys [
         <at 0 1 RCIM_Cubic>
         {arriveTangent=3.2842, interpMode=RCIM_Cubic, leaveTangent=3.2842, time=0.48, value=3}
         {arriveTangent=-4.8077, interpMode=RCIM_Cubic, leaveTangent=-4.8077, time=0.5336, value=2.7526}
         <at 1 0.5>
        ]}
      }})
    ScaleA=true
    ScaleRGB=true
   )
   &DynamicMaterialParameters(
    Index 0 Param 1=1
    Index 0 Param 3=1
    Index 0 Param 4=1
    Module.Index0Float0Write=true
    Module.Index0Float1Write=true
    Module.Index0Float2Write=true
    Module.Index0Float3Write=true
   )
   &CameraOffset(Camera Offset Amount=5)
  ]
  rendererClasses [&NiagaraSpriteRendererProperties]
  renderers [{rendererClass=&NiagaraSpriteRendererProperties}]
 }
 FloatingFrags {
  emitterUpdateScript [
   &EmitterState(Life Cycle Mode=::NewEnumerator1, Loop Behavior=::NewEnumerator1, Loop Duration=2, UseLoopDelay=true)
   &SpawnBurst_Instantaneous(Spawn Count=3, Use Spawn Probability=false, UseLoopCountLimit=false)
  ]
  particleSpawnScript [
   /SetVariables(Particles.SpawnRotationVector=&RandomRangeVector(Minimum=<-0.01 -0.01 -0.01>, Maximum=<0.01 0.01 0.01>, Evaluation Type=::NewEnumerator1)) {isSetParameters=true}
   &InitializeParticle(
    Color=<rgba 1.2351 1.6257 2 0.7>
    Lifetime=&RandomRangeFloat(Minimum=0.7, Maximum=0.8, Evaluation Type=::NewEnumerator1)
    Mesh Scale=&VectorFromFloat(Value=&RandomRangeFloat(Minimum=24, Maximum=27, Evaluation Type=::NewEnumerator1))
    Mesh Scale Mode=::NewEnumerator4
    UsePositionOffset=false
    Write Material Random=true
   )
   &Color(Color=&MakeLinearColorFromVectorAndFloat("Vector (RGB)"=<1.2351 1.6257 2>, "Float (Alpha)"=&ClampFloat(Float=0.7))) {enabled=false}
   &CascadeConversion_InitialMeshRotation(Rotation:Vector3f=&RandomRangeVector(Minimum=<0 0 0>, Maximum=<1 1 1>, Evaluation Type=::NewEnumerator1))
  ]
  particleUpdateScript [
   &ParticleState()
   &UpdateMeshOrientation(Rotation Vector=$Particles.SpawnRotationVector)
   &ScaleMeshSize(Scale Factor=&VectorFromFloat(Value=&FloatFromCurve(FloatCurve=DataInterface() {propertyValues {Curve {keys [
          {arriveTangent=-10.0475, interpMode=RCIM_Cubic, time=0, value=0.8}
          {arriveTangent=0.2701, interpMode=RCIM_Cubic, leaveTangent=0.2701, time=0.1, value=0.4}
          {arriveTangent=1.0796, interpMode=RCIM_Cubic, leaveTangent=1.0796, time=0.329, value=0.8889}
          {arriveTangent=-0.2814, interpMode=RCIM_Cubic, leaveTangent=-0.2814, time=0.5998, value=0.9395}
          {interpMode=RCIM_Cubic, leaveTangent=-1.045, time=1, value=0.7}
         ]}}})))
   &DynamicMaterialParameters(
    Index 0 Param 1=1
    Index 0 Param 2=1
    Index 0 Param 3=1
    Index 0 Param 4=&FloatFromCurve(FloatCurve=DataInterface() {propertyValues {Curve {keys [
         {arriveTangent=-66.5826, interpMode=RCIM_Cubic, time=0, value=-0.5}
         {arriveTangent=2.9197, interpMode=RCIM_Cubic, leaveTangent=2.9197, time=0.1, value=-4}
         {arriveTangent=5.5556, interpMode=RCIM_Cubic, leaveTangent=5.5556, time=0.3314, value=0.4675}
         <at 1 1 RCIM_Cubic>
        ]}}})
    Module.Index0Float0Write=true
    Module.Index0Float1Write=true
    Module.Index0Float2Write=true
    Module.Index0Float3Write=true
   )
   &ScaleColor(
    Scale Alpha=&FloatFromCurve(FloatCurve=DataInterface() {propertyValues {Curve {keys [{arriveTangent=42.9531, interpMode=RCIM_Cubic, time=0, value=0}, <at 0.05 1 RCIM_Cubic>, <at 1 0>]}}})
    Scale RGB=&VectorFromCurve(VectorCurve=DataInterface() {propertyValues {
       XCurve {keys [<at 0 7.0666 RCIM_Cubic>,<at 0.4 2 RCIM_Cubic>]}
       YCurve {keys [<at 0 9.701 RCIM_Cubic>,<at 0.4 2 RCIM_Cubic>]}
       ZCurve {keys [<at 0 10 RCIM_Cubic>,<at 0.4 2 RCIM_Cubic>]}
      }})
    ScaleA=true
    ScaleRGB=true
   )
   &CameraOffset(Camera Offset Amount=200)
  ]
  rendererClasses [&NiagaraMeshRendererProperties]
  renderers [{rendererClass=&NiagaraMeshRendererProperties, rendererIndex=0}]
 }
 GlowStart {
  emitterUpdateScript [
   &EmitterState(Life Cycle Mode=::NewEnumerator1, Loop Behavior=::NewEnumerator1, Loop Duration=1, UseLoopDelay=true)
   &SpawnBurst_Instantaneous(Use Spawn Probability=false, UseLoopCountLimit=false)
  ]
  particleSpawnScript [
   &InitializeParticle(
    Color=<rgba 0.396 0.4194 1 0.1>
    Mesh Scale=<15 15 15>
    Mesh Scale Mode=::NewEnumerator4
    UsePositionOffset=false
    Write Material Random=true
   )
   &Color(Color=&MakeLinearColorFromVectorAndFloat("Vector (RGB)"=<0.396 0.4194 1>, "Float (Alpha)"=&ClampFloat(Float=0.1))) {enabled=false}
  ]
  particleUpdateScript [
   &ParticleState()
   &DynamicMaterialParameters(
    Index 0 Param 1=&FloatFromCurve(FloatCurve=DataInterface() {propertyValues {Curve {keys [<at 0 12>,<at 1 12>]}}})
    Index 0 Param 2=2
    Index 0 Param 3=15
    Index 0 Param 4=1
    Module.Index0Float0Write=true
    Module.Index0Float1Write=true
    Module.Index0Float2Write=true
    Module.Index0Float3Write=true
   )
   &CameraOffset(Camera Offset Amount=2)
   &ScaleColor(
    Scale Alpha=0.7
    Scale RGB=&VectorFromCurve(VectorCurve=DataInterface() {propertyValues {
       XCurve {keys [<at 0 0>,<at 0.5 0.0414>,<at 1 0>]}
       YCurve {keys [<at 0 0>,<at 0.5 0.2389>,<at 1 0>]}
       ZCurve {keys [<at 0 0>,<at 0.5 1>,<at 1 0>]}
      }})
    ScaleA=true
    ScaleRGB=true
   )
  ]
  rendererClasses [&NiagaraMeshRendererProperties]
  renderers [{rendererClass=&NiagaraMeshRendererProperties, rendererIndex=0}]
 }
 Light {
  emitterUpdateScript [
   &EmitterState(Inactive Response=::NewEnumerator1, Life Cycle Mode=::NewEnumerator1, Loop Behavior=::NewEnumerator1, Loop Duration=1, UseLoopDelay=true)
   &SpawnBurst_Instantaneous(Use Spawn Probability=false, UseLoopCountLimit=false)
  ]
  particleSpawnScript [
   &InitializeParticle(
    Color=<rgba 0.3323 1.6388 2 1>
    Sprite Size=<1500 1500>
    Sprite Size Mode=::NewEnumerator4
    UsePositionOffset=false
    Write Material Random=true
   )
   &Color(Color=&MakeLinearColorFromVectorAndFloat("Vector (RGB)"=<0.4984 2.4582 3>, "Float (Alpha)"=&ClampFloat(Float=1))) {enabled=false}
  ]
  particleUpdateScript [
   /SetVariables(Particles.LightExponent=8) {isSetParameters=true}
   &ParticleState()
   &ScaleColor(
    Scale Alpha=&FloatFromCurve(FloatCurve=DataInterface() {propertyValues {Curve {keys [
         {arriveTangent=-1.8703, interpMode=RCIM_Cubic, time=0, value=1}
         {arriveTangent=-1.4917, interpMode=RCIM_Cubic, leaveTangent=-1.4917, time=0.4927, value=0.3962}
         {arriveTangent=5.8576, interpMode=RCIM_Cubic, leaveTangent=5.8576, time=0.5163, value=0.2299}
         {arriveTangent=-8.2655, interpMode=RCIM_Cubic, leaveTangent=-8.2655, time=0.5261, value=0.5923}
         {arriveTangent=-4.3031, interpMode=RCIM_Cubic, leaveTangent=-4.3031, time=0.5374, value=0.0554}
         {arriveTangent=-0.2017, interpMode=RCIM_Cubic, leaveTangent=-0.2017, time=0.5471, value=0.502}
         {arriveTangent=-5.5769, interpMode=RCIM_Cubic, leaveTangent=-5.5769, time=0.5869, value=0.0454}
         {arriveTangent=0.2884, interpMode=RCIM_Cubic, leaveTangent=0.2884, time=0.603, value=0.1904}
         {arriveTangent=-1.0461, interpMode=RCIM_Cubic, leaveTangent=-1.0461, time=0.6279, value=0.0572}
         {arriveTangent=1.4237, interpMode=RCIM_Cubic, leaveTangent=1.4237, time=0.7516, value=0.035}
         {arriveTangent=0.1392, interpMode=RCIM_Cubic, leaveTangent=0.1392, time=0.759, value=0.2438}
         {arriveTangent=-1.5857, interpMode=RCIM_Cubic, leaveTangent=-1.5857, time=0.7769, value=0.0385}
         {arriveTangent=1.1133, interpMode=RCIM_Cubic, leaveTangent=1.1133, time=0.8822, value=0.0484}
         {arriveTangent=0.3163, interpMode=RCIM_Cubic, leaveTangent=0.3163, time=0.8938, value=0.1687}
         {arriveTangent=-1.5889, interpMode=RCIM_Cubic, leaveTangent=-1.5889, time=0.9088, value=0.0569}
         <at 1 0>
        ]}}})
    Scale RGB=&VectorFromFloat(Value=&FloatFromCurve(FloatCurve=DataInterface() {propertyValues {Curve {keys [{arriveTangent=-33.8851, interpMode=RCIM_Cubic, time=0, value=5}, <at 0.4 0.5 RCIM_Cubic>]}}}))
    ScaleA=true
    ScaleRGB=true
   )
   &CascadeConversion_LightProperties(LightBrightnessScale=32, LightColorScale=<1 1 1>, LightRadiusScale=1)
  ]
  rendererClasses [&NiagaraLightRendererProperties]
  renderers [{rendererClass=&NiagaraLightRendererProperties, rendererIndex=0}]
 }
 RoundFlare {
  emitterUpdateScript [
   &EmitterState(Inactive Response=::NewEnumerator1, Life Cycle Mode=::NewEnumerator1, Loop Behavior=::NewEnumerator1, Loop Duration=1, UseLoopDelay=true)
   &SpawnBurst_Instantaneous(Use Spawn Probability=false, UseLoopCountLimit=false)
  ]
  particleSpawnScript [
   &InitializeParticle(
    Color=<rgba 0.4984 2.4582 3 0.25>
    Lifetime=1.2
    Sprite Size=<2200 2200>
    Sprite Size Mode=::NewEnumerator4
    UsePositionOffset=false
    Write Material Random=true
   )
   &Color(Color=&MakeLinearColorFromVectorAndFloat("Vector (RGB)"=<0.4984 2.4582 3>, "Float (Alpha)"=&ClampFloat(Float=0.25))) {enabled=false}
  ]
  particleUpdateScript [
   &ParticleState()
   &ScaleSpriteSize(
    Scale Sprite Size Mode=::NewEnumerator0
    Uniform Scale Factor=&FloatFromCurve(FloatCurve=DataInterface() {propertyValues {Curve {keys [
         {arriveTangent=-2.6785, interpMode=RCIM_Cubic, time=0, value=1}
         {arriveTangent=-1.0154, interpMode=RCIM_Cubic, leaveTangent=-1.0154, time=0.0887, value=0.2526}
         {arriveTangent=-0.2772, interpMode=RCIM_Cubic, leaveTangent=-0.2772, time=0.914, value=0.0719}
         <at 1 0 RCIM_Cubic>
        ]}}})
   )
   &ScaleColor(
    Scale Alpha=&FloatFromCurve(FloatCurve=DataInterface() {propertyValues {Curve {keys [<at 0 1 RCIM_Cubic>,<at 1 0>]}}})
    Scale RGB=&VectorFromFloat(Value=&FloatFromCurve(
      FloatCurve=DataInterface() {propertyValues {Curve {keys [{arriveTangent=-6.3837, interpMode=RCIM_Cubic, time=0, value=10}, <at 1 1 RCIM_Cubic>]}}}
      Scale Curve=0.5
     ))
    ScaleA=true
    ScaleRGB=true
   )
   &DynamicMaterialParameters(
    Index 0 Param 1=1
    Index 0 Param 2=2
    Index 0 Param 3=1
    Index 0 Param 4=1
    Module.Index0Float0Write=true
    Module.Index0Float1Write=true
    Module.Index0Float2Write=true
    Module.Index0Float3Write=true
   )
   &CameraOffset(Camera Offset Amount=300)
  ]
  rendererClasses [&NiagaraSpriteRendererProperties]
  renderers [{rendererClass=&NiagaraSpriteRendererProperties}]
 }
 ShockWave {
  emitterUpdateScript [
   &EmitterState(Life Cycle Mode=::NewEnumerator1, Loop Behavior=::NewEnumerator1, Loop Duration=1, UseLoopDelay=true)
   &SpawnBurst_Instantaneous(Use Spawn Probability=false, UseLoopCountLimit=false)
  ]
  particleSpawnScript [
   &InitializeParticle(
    Color=<rgba 1 1 1 0.05>
    Lifetime=0.6
    Mesh Scale=<2 2 2>
    Mesh Scale Mode=::NewEnumerator4
    UsePositionOffset=false
    Write Material Random=true
   )
   &Color(Color=&MakeLinearColorFromVectorAndFloat("Vector (RGB)"=<1 1 1>, "Float (Alpha)"=&ClampFloat(Float=0.05))) {enabled=false}
  ]
  particleUpdateScript [
   &ParticleState()
   &ScaleMeshSize(Scale Factor=&VectorFromCurve(VectorCurve=DataInterface() {propertyValues {XCurve {keys [<at 0 0.2>,<at 1 8>]}, YCurve {keys [<at 0 0.2>,<at 1 8>]}, ZCurve {keys [<at 0 0.2>,<at 1 8>]}}}))
   &ScaleColor(
    Scale Alpha=&FloatFromCurve(FloatCurve=DataInterface() {propertyValues {Curve {keys [{arriveTangent=-2.1694, interpMode=RCIM_Cubic, time=0, value=1}, <at 1 0 RCIM_Cubic>]}}})
    ScaleA=true
    ScaleRGB=true
   )
   &CameraOffset(Camera Offset Amount=400)
  ]
  rendererClasses [&NiagaraMeshRendererProperties]
  renderers [{rendererClass=&NiagaraMeshRendererProperties, rendererIndex=0}]
 }
 WideFlare {
  emitterUpdateScript [
   &EmitterState(Inactive Response=::NewEnumerator1, Life Cycle Mode=::NewEnumerator1, Loop Behavior=::NewEnumerator1, Loop Duration=1, UseLoopDelay=true)
   &SpawnBurst_Instantaneous(Use Spawn Probability=false, UseLoopCountLimit=false)
  ]
  particleSpawnScript [
   &InitializeParticle(
    Color=<rgba 0.4984 2.4582 3 1>
    Lifetime=0.7
    Sprite Size=<7000 200>
    Sprite Size Mode=::NewEnumerator4
    UsePositionOffset=false
    Write Material Random=true
   )
   &Color(Color=&MakeLinearColorFromVectorAndFloat("Vector (RGB)"=<0.4984 2.4582 3>, "Float (Alpha)"=&ClampFloat(Float=1))) {enabled=false}
  ]
  particleUpdateScript [
   &ParticleState()
   &ScaleSpriteSize(
    Scale Factor=&Vector2DFromCurve(
     Vector2Curve=DataInterface() {propertyValues {
       XCurve {keys [{arriveTangent=-2.6785, interpMode=RCIM_Cubic, time=0, value=1}, <at 1 0 RCIM_Cubic>]}
       YCurve {keys [<at 0 1 RCIM_Cubic>,<at 1 0 RCIM_Cubic>]}
      }}
     ScaleCurve=<1 1>
    )
    Scale Sprite Size Mode=::NewEnumerator2
   )
   &ScaleColor(
    Scale RGB=&VectorFromFloat(Value=&FloatFromCurve(
      FloatCurve=DataInterface() {propertyValues {Curve {keys [{arriveTangent=-6.3837, interpMode=RCIM_Cubic, time=0, value=5}, <at 1 1 RCIM_Cubic>]}}}
      Scale Curve=0.5
     ))
    ScaleA=true
    ScaleRGB=true
   )
   &DynamicMaterialParameters(
    Index 0 Param 1=1
    Index 0 Param 2=2
    Index 0 Param 3=1
    Index 0 Param 4=1
    Module.Index0Float0Write=true
    Module.Index0Float1Write=true
    Module.Index0Float2Write=true
    Module.Index0Float3Write=true
   )
   &CameraOffset(Camera Offset Amount=300)
  ]
  rendererClasses [&NiagaraSpriteRendererProperties]
  renderers [{rendererClass=&NiagaraSpriteRendererProperties}]
 }
}
summary {
 systemName=NS_Bling_Van_Portal
 emitters [
  {emitterName=CrackSpline, rendererClasses [&NiagaraMeshRendererProperties]}
  {emitterName=FloatingFrags, rendererClasses [&NiagaraMeshRendererProperties]}
  {emitterName=ElectricArcsResidual, rendererClasses [&NiagaraMeshRendererProperties]}
  {emitterName=WideFlare, rendererClasses [&NiagaraSpriteRendererProperties]}
  {emitterName=RoundFlare, rendererClasses [&NiagaraSpriteRendererProperties]}
  {emitterName=Light, rendererClasses [&NiagaraLightRendererProperties]}
  {emitterName=Corona, rendererClasses [&NiagaraSpriteRendererProperties]}
  {emitterName=ShockWave, rendererClasses [&NiagaraMeshRendererProperties]}
  {emitterName=GlowStart, rendererClasses [&NiagaraMeshRendererProperties]}
  {emitterName=Comet, rendererClasses [&NiagaraMeshRendererProperties]}
  {emitterName=Flare, rendererClasses [&NiagaraSpriteRendererProperties]}
 ]
 userVariables []
}
