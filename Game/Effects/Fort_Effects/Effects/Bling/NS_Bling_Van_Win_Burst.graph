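# Niagara one-shot burst effect for the Bling van win moment, centered slightly above the system origin.
# Mesh glow, shell, wipe and warm firework emitters layer with blue sprite flares, shimmers and teal GPU sparks that fly outward and fall.
# Flare and wipe emitters use spawn groups to vary two bursts within one emitter.
# Curves drive scale, color fades and dynamic material parameters, with no user parameters exposed.
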NiagaraSystem /Game/Effects/Fort_Effects/Effects/Bling/NS_Bling_Van_Win_Burst
dependencies {
 usedDataInterfaces [
  &NiagaraDataInterfaceCurve
  &NiagaraDataInterfaceVectorCurve
  &NiagaraDataInterfaceVector2DCurve
 ]
 usedDynamicInputs [&ConvertVectorToPosition]
 usedModules [&SystemState]
}
emitters {
 FlashFlare {
  emitterUpdateScript [
   &EmitterState(Life Cycle Mode=::NewEnumerator1, Loop Behavior=::NewEnumerator1, Loop Duration=1, UseLoopDelay=true)
   &SpawnBurst_Instantaneous(Use Spawn Probability=false, UseLoopCountLimit=false)
   &SpawnBurst_Instantaneous(Spawn Group=1, Use Spawn Probability=false, UseLoopCountLimit=false) {module=SpawnBurst_Instantaneous001}
  ]
  particleSpawnScript [
   &InitializeParticle(
    Color=&MergeHelper_ColorSwitch(0=<rgba 0 0.25 1.5 1>, 1=<rgba 0 5 25 5>, 2=?, 3=?, Index=&SpawnGroup(), Num=::NewEnumerator0)
    Lifetime=&MergeHelper_FloatSwitch(0=0.15, 1=0.35, 2=?, 3=?, Index=&SpawnGroup(), Num=::NewEnumerator0)
    Position Offset=<0 0 100>
    Sprite Size=&MergeHelper_Vector2Switch(0=<1250 1250>, 1=<2500 2500>, 2=?, 3=?, 4=?, Index=&SpawnGroup(), Num=::NewEnumerator0)
    Sprite Size Mode=::NewEnumerator4
    UsePositionOffset=true
    Write Material Random=true
   )
   &DynamicMaterialParameters(
    Index 0 Param 1=1
    Module.Index0Float0Write=true
    Module.Index0Float1Write=false
    Module.Index0Float2Write=false
    Module.Index0Float3Write=false
   )
   /SetVariables(Particles.VisibilityTag=&MergeHelper_IntSwitch(0=0, 1=1, 2=?, 3=?, 4=?, 5=?, Index=&SpawnGroup(), Num=::NewEnumerator0)) {isSetParameters=true}
  ]
  particleUpdateScript [
   &ParticleState()
   &ScaleSpriteSize(
    Scale Sprite Size Mode=::NewEnumerator0
    Uniform Scale Factor=&MergeHelper_FloatSwitch(
     0=&FloatFromCurve(FloatCurve=DataInterface() {propertyValues {Curve {keys [<at 0 1>,<at 1 3>]}}})
     1=&FloatFromCurve(FloatCurve=DataInterface() {propertyValues {Curve {keys [{arriveTangent=2.6788, interpMode=RCIM_Cubic, time=0, value=0}, <at 1 1>]}}})
     2=?
     3=?
     Index=&SpawnGroup()
     Num=::NewEnumerator0
    )
   )
   &ScaleColor(
    Scale Alpha=&MergeHelper_FloatSwitch(
     0=&FloatFromCurve(FloatCurve=DataInterface() {propertyValues {Curve {keys [{arriveTangent=-1.2898, interpMode=RCIM_Cubic, leaveTangent=-1.2898, tangentMode=RCTM_User, time=0, value=1}, <at 1 0>]}}})
     1=&FloatFromCurve(FloatCurve=DataInterface() {propertyValues {Curve {keys [
          <at -0.0045 0.0182 RCIM_Cubic>
          {arriveTangent=3.8563, interpMode=RCIM_Cubic, leaveTangent=3.8563, time=0.0333, value=1}
          {arriveTangent=-2.2133, interpMode=RCIM_Cubic, leaveTangent=-2.2133, time=0.1601, value=0.6529}
          {arriveTangent=-0.7773, interpMode=RCIM_Cubic, leaveTangent=-0.7773, time=0.3612, value=0.2742}
          <at 1 0>
         ]}}})
     2=?
     3=?
     Index=&SpawnGroup()
     Num=::NewEnumerator0
    )
    Scale RGB=&MergeHelper_VectorSwitch(
     0=<4 4 4>
     1=&VectorFromCurve(VectorCurve=DataInterface() {propertyValues {
        XCurve {keys [<at 0 1 RCIM_Cubic>,<at 1 0.3728>]}
        YCurve {keys [<at 0 0.6851 RCIM_Cubic>,<at 1 1>]}
        ZCurve {keys [<at 0 0.8647 RCIM_Cubic>,<at 1 0.6644>]}
       }})
     2=?
     3=?
     Index=&SpawnGroup()
     Num=::NewEnumerator0
    )
    ScaleA=true
    ScaleRGB=true
   )
   &CameraOffset(Camera Offset Amount=&MergeHelper_FloatSwitch(0=500, 1=250, 2=?, 3=?, Index=&SpawnGroup(), Num=::NewEnumerator0))
   &DynamicMaterialParameters(
    Index 0 Param 2=1
    Index 0 Param 3=1
    Index 0 Param 4=1
    Module.Index0Float0Write=false
    Module.Index0Float1Write=true
    Module.Index0Float2Write=true
    Module.Index0Float3Write=true
   ) {module=DynamicMaterialParameters001}
   /SetVariables(Particles.DP0:Vector4f=&MakeVector4(
     X=&FloatFromCurve(FloatCurve=DataInterface() {propertyValues {Curve {keys [{arriveTangent=-1.2514, interpMode=RCIM_Cubic, time=0, value=1.6}, <at 1 0.5>]}}})
     Y=1
     Z=0
     W=1
    )) {isSetParameters=true}
  ]
  rendererClasses [&NiagaraSpriteRendererProperties]
  renderers [
   {rendererClass=&NiagaraSpriteRendererProperties}
   {rendererClass=&NiagaraSpriteRendererProperties, rendererIndex=1}
  ]
 }
 GlowStart {
  emitterUpdateScript [
   &EmitterState(Life Cycle Mode=::NewEnumerator1, Loop Behavior=::NewEnumerator1, Loop Duration=1, UseLoopDelay=true)
   &SpawnBurst_Instantaneous(Use Spawn Probability=false, UseLoopCountLimit=false)
  ]
  particleSpawnScript [&InitializeParticle(
    Color=<rgba 0 0.2374 1 0.25>
    Lifetime=0.35
    Mesh Scale=<35 35 35>
    Mesh Scale Mode=::NewEnumerator4
    Position Offset=<0 0 100>
    UsePositionOffset=true
    Write Material Random=true
   )]
  particleUpdateScript [
   &ParticleState()
   &ScaleMeshSize(Scale Factor=&VectorFromFloat(Value=&FloatFromCurve(FloatCurve=DataInterface() {propertyValues {Curve {keys [<at 0 0.6>,<at 1 1>]}}})))
   &DynamicMaterialParameters(
    Index 0 Param 1=&FloatFromCurve(FloatCurve=DataInterface() {propertyValues {Curve {keys [<at 0 15>,<at 1 18>]}}})
    Index 0 Param 2=2
    Index 0 Param 3=15
    Index 0 Param 4=1
    Module.Index0Float0Write=true
    Module.Index0Float1Write=true
    Module.Index0Float2Write=true
    Module.Index0Float3Write=true
   )
   &CameraOffset(Camera Offset Amount=2)
   &ScaleColor(
    Scale Alpha=&FloatFromCurve(FloatCurve=DataInterface() {propertyValues {Curve {keys [<at 0 1>,<at 1 0>]}}})
    Scale RGB=&VectorFromFloat(Value=&FloatFromCurve(FloatCurve=DataInterface() {propertyValues {Curve {keys [<at 0 1 RCIM_Cubic>,<at 0.9961 0 RCIM_Cubic>]}}}))
    ScaleA=true
    ScaleRGB=true
   )
  ]
  rendererClasses [&NiagaraMeshRendererProperties]
  renderers [{rendererClass=&NiagaraMeshRendererProperties, rendererIndex=0}]
 }
 LOW_Sparks {
  simTarget=GPUComputeSim
  emitterUpdateScript [
   &EmitterState(Life Cycle Mode=::NewEnumerator1, Loop Behavior=::NewEnumerator1, Loop Duration=0.1, UseLoopDelay=true)
   &SpawnBurst_Instantaneous(
    Spawn Count=&RandomRangeInt(Minimum=150, Maximum=250)
    Use Spawn Probability=false
    UseLoopCountLimit=false
   )
   &SpawnRate(SpawnRate=100, Use Spawn Probability=false)
  ]
  particleSpawnScript [
   &InitializeParticle(
    Color=<rgba 0 800 500 1>
    Lifetime=&RandomRangeFloat(Minimum=0.25, Maximum=2, Evaluation Type=::NewEnumerator1)
    Position Offset=<0 0 100>
    Sprite Rotation Angle=&RandomRangeFloat(Minimum=0.02, Maximum=-0.02, Evaluation Type=::NewEnumerator1)
    Sprite Rotation Mode=::NewEnumerator1
    Sprite Size=&Vector2DFromFloat(Value=&RandomRangeFloat(Minimum=10, Maximum=20, Evaluation Type=::NewEnumerator1))
    Sprite Size Mode=::NewEnumerator4
    UsePositionOffset=true
    Write Material Random=true
   )
   &ShapeLocation(
    MaskSpawnGroup=false
    Override Seed=false
    Sphere Radius=&FloatFromCurve(FloatCurve=DataInterface() {propertyValues {Curve {keys [<at 0 250>,<at 0.1 1000>]}}}, CurveIndex=$Emitter.Age, Scale Curve=0.75)
    Sphere Surface Distribution=1
    UseOwnerScale=true
   )
   &AddVelocity(Velocity=&Multiply_VectorByFloat(
     Vector=$Particles.ShapeLocation.ShapeVector
     Float=&FloatFromCurve(
      FloatCurve=DataInterface() {propertyValues {Curve {keys [{arriveTangent=-51.1112, interpMode=RCIM_Cubic, time=-0.0001, value=2.0012}, <at 0.1 0.25 RCIM_Cubic>]}}}
      CurveIndex=$Emitter.LoopedAge
     )
    ))
   &AddVelocity(Velocity=&RandomRangeVector(Minimum=<0 0 0>, Maximum=<0 0 250>, Evaluation Type=::NewEnumerator1)) {module=AddVelocity001}
  ]
  particleUpdateScript [
   &ParticleState()
   &ScaleSpriteSize(
    Scale Sprite Size Mode=::NewEnumerator0
    Uniform Scale Factor=&FloatFromCurve(FloatCurve=DataInterface() {propertyValues {Curve {keys [
         {arriveTangent=-0.421, interpMode=RCIM_Cubic, time=-0.0052, value=0.9944}
         {arriveTangent=-0.9892, interpMode=RCIM_Cubic, leaveTangent=-0.9892, time=0.3594, value=0.8427}
         <at 1 0 RCIM_Cubic>
        ]}}})
   )
   &ScaleColor(
    Scale Alpha=&FloatFromCurve(FloatCurve=DataInterface() {propertyValues {Curve {keys [
         {arriveTangent=-0.7236, interpMode=RCIM_Cubic, time=0, value=0.0994}
         {arriveTangent=-0.0994, interpMode=RCIM_Cubic, leaveTangent=-0.0994, time=0.128, value=0.0373}
         <at 1 0 RCIM_Cubic>
        ]}}})
    Scale RGB=&VectorFromFloat(Value=&FloatFromCurve(FloatCurve=DataInterface() {propertyValues {Curve {keys [{arriveTangent=-14.4521, interpMode=RCIM_Cubic, time=0, value=15}, <at 0.4999 1.0559 RCIM_Cubic>]}}}))
    ScaleA=true
    ScaleRGB=true
   )
   &AccelerationForce(Acceleration=<0 0 -1200>)
   &SolveForcesAndVelocity(Clamp Velocity=false, Limit Acceleration=false)
  ]
  rendererClasses [&NiagaraSpriteRendererProperties]
  renderers [{rendererClass=&NiagaraSpriteRendererProperties}]
 }
 MIDHIGH_Firework {
  emitterUpdateScript [
   &EmitterState(Life Cycle Mode=::NewEnumerator1, Loop Behavior=::NewEnumerator1, Loop Duration=1, UseLoopDelay=true)
   &SpawnBurst_Instantaneous(Spawn Count=2, Use Spawn Probability=false, UseLoopCountLimit=false)
  ]
  particleSpawnScript [
   /SetVariables(Particles.Acceleration:Vector3f=&RandomRangeVector(Minimum=<0 0 -100>, Maximum=<0 0 -150>, Evaluation Type=::NewEnumerator1)) {isSetParameters=true}
   /SetVariables(Particles.SpawnRotationVector=&RandomRangeVector(Minimum=<-0.01 -0.01 -0.01>, Maximum=<0.01 0.01 0.01>, Evaluation Type=::NewEnumerator1)) {isSetParameters=true}
   &InitializeParticle(
    Color=<rgba 0.8 0.25 0 1>
    Lifetime=&RandomRangeFloat(Minimum=0.45, Maximum=1, Evaluation Type=::NewEnumerator1)
    Mesh Scale Mode=::NewEnumerator3
    Mesh Uniform Scale=0.2
    Position Offset=<0 0 100>
    UsePositionOffset=true
    Write Material Random=true
   )
   &Color(Color=&MakeLinearColorFromVectorAndFloat("Vector (RGB)"=<0.8 0.25 0>, "Float (Alpha)"=&ClampFloat(Float=1))) {enabled=false}
   /Niagara/Modules/CascadeConversion/CascadeConversion_InitialMeshRotation(Rotation:Vector3f=&RandomRangeVector(Minimum=<0 0 0>, Maximum=<1 1 1>, Evaluation Type=::NewEnumerator1))
  ]
  particleUpdateScript [
   &ParticleState()
   &UpdateMeshOrientation(Rotation Vector=$Particles.SpawnRotationVector)
   &ScaleMeshSize(Scale Factor=&VectorFromFloat(Value=&FloatFromCurve(FloatCurve=DataInterface() {propertyValues {Curve {keys [{arriveTangent=1.6602, interpMode=RCIM_Cubic, leaveTangent=1.6602, tangentMode=RCTM_User, time=0, value=0.8}, <at 1 2 RCIM_Cubic>]}}})))
   &DynamicMaterialParameters(
    Index 0 Param 1=1
    Index 0 Param 2=1
    Index 0 Param 3=&FloatFromCurve(
     FloatCurve=DataInterface() {propertyValues {Curve {keys [{arriveTangent=13.2441, interpMode=RCIM_Cubic, time=0.038, value=10}, <at 1.0021 14.9234 RCIM_Cubic>]}}}
     Scale Curve=0.5
    )
    Index 0 Param 4=&FloatFromCurve(FloatCurve=DataInterface() {propertyValues {Curve {keys [
         <at 0.0098 1.1329 RCIM_Cubic>
         {arriveTangent=-1.1265, interpMode=RCIM_Cubic, leaveTangent=-1.1265, time=0.5142, value=0.2716}
         <at 1.0011 0.0162 RCIM_Cubic>
        ]}}})
    Module.Index0Float0Write=true
    Module.Index0Float1Write=true
    Module.Index0Float2Write=true
    Module.Index0Float3Write=true
   )
   &ScaleColor(
    Scale Alpha=&FloatFromCurve(FloatCurve=DataInterface() {propertyValues {Curve {keys [<at 0 1>,<at 1 0>]}}}, Scale Curve=1.6)
    Scale RGB=&VectorFromCurve(VectorCurve=DataInterface() {propertyValues {
       XCurve {keys [
         <at 0 1 RCIM_Cubic>
         {arriveTangent=-0.6013, interpMode=RCIM_Cubic, leaveTangent=-0.6013, time=0.2556, value=1}
         {arriveTangent=-0.6727, interpMode=RCIM_Cubic, leaveTangent=-0.6727, time=0.4989, value=0.7}
         <at 0.9989 0.5 RCIM_Cubic>
        ]}
       YCurve {keys [
         {arriveTangent=-93.5811, interpMode=RCIM_Cubic, time=0, value=25}
         {arriveTangent=-43.3962, interpMode=RCIM_Cubic, leaveTangent=-43.3962, time=0.2556, value=8.0913}
         {arriveTangent=-8.6657, interpMode=RCIM_Cubic, leaveTangent=-8.6657, time=0.4989, value=3.3502}
         <at 0.9989 1.6498 RCIM_Cubic>
        ]}
       ZCurve {keys [<at 0 1 RCIM_Cubic>,<at 0.2556 1 RCIM_Cubic>,<at 0.4989 1 RCIM_Cubic>,<at 0.9989 1 RCIM_Cubic>]}
      }})
    ScaleA=true
    ScaleRGB=true
   )
   &AccelerationForce(Acceleration=$Particles.Acceleration)
   &SolveForcesAndVelocity(Clamp Velocity=false, Limit Acceleration=false)
  ]
  rendererClasses [&NiagaraMeshRendererProperties]
  renderers [{rendererClass=&NiagaraMeshRendererProperties, rendererIndex=0}]
 }
 Particle_Emitter {
  emitterUpdateScript [
   &EmitterState(Life Cycle Mode=::NewEnumerator1, Loop Behavior=::NewEnumerator1, Loop Duration=1, UseLoopDelay=true)
   &SpawnBurst_Instantaneous(Use Spawn Probability=false, UseLoopCountLimit=false)
  ]
  particleSpawnScript [
   &InitializeParticle(
    Color=<rgba 0 2 205 15>
    Lifetime=0.5
    Mesh Scale=<30 30 30>
    Mesh Scale Mode=::NewEnumerator4
    Position Offset=<0 0 100>
    UsePositionOffset=true
    Write Material Random=true
   )
   &Color(Color=&MakeLinearColorFromVectorAndFloat("Vector (RGB)"=<0 2 205>, "Float (Alpha)"=&ClampFloat(Float=15))) {enabled=false}
  ]
  particleUpdateScript [
   &ParticleState()
   &ScaleMeshSize(Scale Factor=&VectorFromCurve(VectorCurve=DataInterface() {propertyValues {XCurve {keys [<at 0 1>,<at 1 2>]}, YCurve {keys [<at 0 0.5>,<at 1 3>]}, ZCurve {keys [<at 0 1>,<at 1 2>]}}}))
   &ScaleColor(
    Scale Alpha=&FloatFromCurve(FloatCurve=DataInterface() {propertyValues {Curve {keys [<at 0 0>,<at 0.15 1>,<at 1 0>]}}})
    Scale RGB=&VectorFromFloat(Value=&FloatFromCurve(FloatCurve=DataInterface() {propertyValues {Curve {keys [
          {arriveTangent=-8.4361, interpMode=RCIM_Cubic, time=0, value=3}
          {arriveTangent=-3, interpMode=RCIM_Cubic, leaveTangent=-3, time=0.3666, value=0.7094}
          <at 1 0 RCIM_Cubic>
         ]}}}))
    ScaleA=true
    ScaleRGB=true
   )
   &DynamicMaterialParameters(
    Index 0 Param 2=&FloatFromCurve(FloatCurve=DataInterface() {propertyValues {Curve {keys [{arriveTangent=2.1814, interpMode=RCIM_Cubic, time=0.0051, value=0.0052}, <at 1 1>]}}})
    Index 0 Param 3=1
    Index 0 Param 4=1
    Module.Index0Float0Write=true
    Module.Index0Float1Write=true
    Module.Index0Float2Write=true
    Module.Index0Float3Write=true
   )
   &CameraOffset(Camera Offset Amount=250)
  ]
  rendererClasses [&NiagaraMeshRendererProperties]
  renderers [{rendererClass=&NiagaraMeshRendererProperties, rendererIndex=0}]
 }
 Shimmers {
  emitterUpdateScript [
   &EmitterState(Life Cycle Mode=::NewEnumerator1, Loop Behavior=::NewEnumerator1, Loop Duration=0.1, UseLoopDelay=true)
   &SpawnRate(SpawnRate=125, Use Spawn Probability=false)
  ]
  particleSpawnScript [
   /SetVariables(Particles.SpriteRotationRate=&RandomRangeFloat(Minimum=-0.03, Maximum=0.03, Evaluation Type=::NewEnumerator1)) {isSetParameters=true}
   &InitializeParticle(
    Color=&MultiplyLinearColor(
     Linear Color=&RandomRangeLinearColor(Minimum=<rgba 0.1117 4 50 1>, Maximum=<rgba 10.2984 49.5408 125 1>)
     Scale Factor=/Niagara/DynamicInputs/UniformRange/V2/RandomRangeVector4(
      Minimum:Vector4f=<1 1 1 0.1>
      Maximum:Vector4f=<1 1 1 0.25>
      Evaluation Type:ENiagaraRandomnessEvaluation=::NewEnumerator0
      Recalculate Random Each Loop:NiagaraBool=true
      Randomness Mode:ENiagaraRandomnessMode=::NewEnumerator0
      Random Seed:NiagaraInt32=0
      Fixed Random Seed:NiagaraBool=false
     )
    )
    Lifetime=&RandomRangeFloat(Minimum=0.1, Maximum=0.3, Evaluation Type=::NewEnumerator1)
    Position Offset=<0 0 100>
    Sprite Rotation Angle=&RandomRangeFloat(Minimum=0, Maximum=1, Evaluation Type=::NewEnumerator1)
    Sprite Rotation Mode=::NewEnumerator1
    Sprite Size=&RandomRangeVector2D(Minimum=<200 1000>, Maximum=<500 1500>, Evaluation Type=::NewEnumerator1)
    Sprite Size Mode=::NewEnumerator4
    UsePositionOffset=true
    Write Material Random=true
   )
   &Color(Color=&MakeLinearColorFromVectorAndFloat(
     "Vector (RGB)"=&RandomRangeVector(Minimum=<0.1117 4 50>, Maximum=<10.2984 49.5408 125>, Evaluation Type=::NewEnumerator1)
     "Float (Alpha)"=&ClampFloat(Float=&RandomRangeFloat(Minimum=0.1, Maximum=0.25, Evaluation Type=::NewEnumerator1))
    )) {enabled=false}
   &ShapeLocation(
    MaskSpawnGroup=false
    Override Seed=false
    Sphere Radius=&RandomRangeFloat(Minimum=200, Maximum=400, Evaluation Type=::NewEnumerator1)
    UseOwnerScale=true
   )
   &AddVelocity(Velocity=&Multiply_VectorByFloat(Vector=$Particles.ShapeLocation.ShapeVector, Float=1))
   &DynamicMaterialParameters(
    Index 0 Param 1=&RandomRangeFloat(Minimum=0, Maximum=1, Evaluation Type=::NewEnumerator1)
    Index 0 Param 2=&RandomRangeFloat(Minimum=0, Maximum=1, Evaluation Type=::NewEnumerator1)
    Module.Index0Float0Write=true
    Module.Index0Float1Write=true
    Module.Index0Float2Write=false
    Module.Index0Float3Write=false
   )
   &CameraOffset(Camera Offset Amount=&RandomRangeFloat(Minimum=-250, Maximum=350, Evaluation Type=::NewEnumerator1))
  ]
  particleUpdateScript [
   &ParticleState()
   &ScaleSpriteSize(
    Scale Factor=&Vector2DFromCurve(Vector2Curve=DataInterface() {propertyValues {XCurve {keys [<at 0 1>,<at 1 1>]}, YCurve {keys [<at 0 0.25>,<at 1 2>]}}}, ScaleCurve=<1 1>)
    Scale Sprite Size Mode=::NewEnumerator2
   )
   &SpriteRotationRate(Rotation Rate=&AngleConversion(
     Angle:NiagaraFloat=$Particles.SpriteRotationRate
     Angle Input:ENiagara_AngleInput=::NewEnumerator1
     Angle Output:ENiagara_AngleInput=::NewEnumerator0
    ))
   &SolveForcesAndVelocity(Clamp Velocity=false, Limit Acceleration=false)
   &DynamicMaterialParameters(
    Index 0 Param 4=1
    Module.Index0Float0Write=false
    Module.Index0Float1Write=false
    Module.Index0Float2Write=true
    Module.Index0Float3Write=true
   ) {module=DynamicMaterialParameters001}
   &ScaleColor(
    Scale Alpha=&FloatFromCurve(FloatCurve=DataInterface() {propertyValues {Curve {keys [<at 0 0 RCIM_Cubic>,<at 0.0651 1.0049 RCIM_Cubic>,<at 1 0>]}}})
    ScaleA=true
    ScaleRGB=true
   )
  ]
  rendererClasses [&NiagaraSpriteRendererProperties]
  renderers [{rendererClass=&NiagaraSpriteRendererProperties}]
 }
 Sparks {
  simTarget=GPUComputeSim
  emitterUpdateScript [
   &EmitterState(Life Cycle Mode=::NewEnumerator1, Loop Behavior=::NewEnumerator1, Loop Duration=0.1, UseLoopDelay=true)
   &SpawnBurst_Instantaneous(
    Spawn Count=&RandomRangeInt(Minimum=150, Maximum=250)
    Use Spawn Probability=false
    UseLoopCountLimit=false
   )
   &SpawnRate(SpawnRate=2000, Use Spawn Probability=false)
  ]
  particleSpawnScript [
   &InitializeParticle(
    Lifetime=&RandomRangeFloat(Minimum=0.25, Maximum=2, Evaluation Type=::NewEnumerator1)
    Position Offset=<0 0 100>
    Sprite Rotation Angle=&RandomRangeFloat(Minimum=0.02, Maximum=-0.02, Evaluation Type=::NewEnumerator1)
    Sprite Rotation Mode=::NewEnumerator1
    Sprite Size=&Vector2DFromFloat(Value=&RandomRangeFloat(Minimum=4, Maximum=15, Evaluation Type=::NewEnumerator1))
    Sprite Size Mode=::NewEnumerator4
    UsePositionOffset=true
    Write Material Random=true
   )
   &Color(Color=&MakeLinearColorFromVectorAndFloat("Vector (RGB)"=<0 800 500>, "Float (Alpha)"=&ClampFloat(Float=1)))
   &ShapeLocation(
    MaskSpawnGroup=false
    Override Seed=false
    Sphere Radius=&FloatFromCurve(FloatCurve=DataInterface() {propertyValues {Curve {keys [<at 0 250>,<at 0.1 1000>]}}}, CurveIndex=$Emitter.Age, Scale Curve=0.75)
    Sphere Surface Distribution=1
    UseOwnerScale=true
   )
   &AddVelocity(Velocity=&Multiply_VectorByFloat(
     Vector=$Particles.ShapeLocation.ShapeVector
     Float=&FloatFromCurve(
      FloatCurve=DataInterface() {propertyValues {Curve {keys [{arriveTangent=-51.1112, interpMode=RCIM_Cubic, time=-0.0001, value=2.0012}, <at 0.1 0.25 RCIM_Cubic>]}}}
      CurveIndex=$Emitter.LoopedAge
     )
    ))
   &AddVelocity(Velocity=&RandomRangeVector(Minimum=<0 0 0>, Maximum=<0 0 250>, Evaluation Type=::NewEnumerator1)) {module=AddVelocity001}
  ]
  particleUpdateScript [
   &ParticleState()
   &ScaleSpriteSize(
    Scale Sprite Size Mode=::NewEnumerator0
    Uniform Scale Factor=&FloatFromCurve(FloatCurve=DataInterface() {propertyValues {Curve {keys [
         {arriveTangent=-0.421, interpMode=RCIM_Cubic, time=-0.0052, value=0.9944}
         {arriveTangent=-0.9892, interpMode=RCIM_Cubic, leaveTangent=-0.9892, time=0.3594, value=0.8427}
         <at 1 0 RCIM_Cubic>
        ]}}})
   )
   &ScaleColor(
    Scale Alpha=&FloatFromCurve(FloatCurve=DataInterface() {propertyValues {Curve {keys [
         {arriveTangent=-0.7236, interpMode=RCIM_Cubic, time=0, value=0.0994}
         {arriveTangent=-0.0994, interpMode=RCIM_Cubic, leaveTangent=-0.0994, time=0.128, value=0.0373}
         <at 1 0 RCIM_Cubic>
        ]}}})
    Scale RGB=&VectorFromFloat(Value=&FloatFromCurve(FloatCurve=DataInterface() {propertyValues {Curve {keys [{arriveTangent=-14.4521, interpMode=RCIM_Cubic, time=0, value=15}, <at 0.4999 1.0559 RCIM_Cubic>]}}}))
    ScaleA=true
    ScaleRGB=true
   )
   &AccelerationForce(Acceleration=<0 0 -1200>)
   &SolveForcesAndVelocity(Clamp Velocity=false, Limit Acceleration=false)
  ]
  rendererClasses [&NiagaraSpriteRendererProperties]
  renderers [{rendererClass=&NiagaraSpriteRendererProperties}]
 }
 Wipe {
  emitterUpdateScript [
   &EmitterState(Life Cycle Mode=::NewEnumerator1, Loop Behavior=::NewEnumerator1, Loop Duration=1, UseLoopDelay=true)
   &SpawnBurst_Instantaneous(Spawn Group=1, Use Spawn Probability=false, UseLoopCountLimit=false) {module=SpawnBurst_Instantaneous001}
   &SpawnBurst_Instantaneous(Use Spawn Probability=false, UseLoopCountLimit=false)
  ]
  particleSpawnScript [
   &InitializeParticle(
    Color=<rgba 0.7461 1.3096 3 1>
    Mesh Scale=<25 10 10>
    Mesh Scale Mode=::NewEnumerator4
    Position Offset=<0 0 100>
    UsePositionOffset=true
    Write Material Random=true
   )
   &AddVelocity(Velocity=<0 0 150>)
   &InitialMeshOrientation(
    Mesh Axis To Orient=&MergeHelper_VectorSwitch(0=<0 0 1>, 1=<0 0 -1>, 2=?, 3=?, Index=&SpawnGroup(), Num=::NewEnumerator0)
    Mesh Orientation Mode=::NewEnumerator2
    Use Rotation Vector=false
   )
  ]
  particleUpdateScript [
   &ParticleState()
   &ScaleMeshSize(Scale Factor=&VectorFromCurve(VectorCurve=DataInterface() {propertyValues {
       XCurve {keys [{arriveTangent=5.9162, interpMode=RCIM_Cubic, time=0, value=0.25}, {interpMode=RCIM_Cubic, leaveTangent=1.4151, time=0.25, value=1}]}
       YCurve {keys [<at 0 1 RCIM_Cubic>,<at 0.25 1 RCIM_Cubic>]}
       ZCurve {keys [<at 0 1 RCIM_Cubic>,<at 0.25 1 RCIM_Cubic>]}
      }}))
   &DynamicMaterialParameters(
    Index 0 Param 1=1
    Index 0 Param 2=1
    Index 0 Param 3=&FloatFromCurve(FloatCurve=DataInterface() {propertyValues {Curve {keys [<at 0 1>,<at 0.3949 1>,<at 1 1>]}}})
    Index 0 Param 4=1
    Module.Index0Float0Write=true
    Module.Index0Float1Write=true
    Module.Index0Float2Write=true
    Module.Index0Float3Write=true
   )
   &CameraOffset(Camera Offset Amount=15)
   &ScaleColor(
    Scale Alpha=&FloatFromCurve(FloatCurve=DataInterface() {propertyValues {Curve {keys [
         {arriveTangent=-0.0791, interpMode=RCIM_Cubic, time=-0.0013, value=0.0045}
         {arriveTangent=0.6118, interpMode=RCIM_Cubic, leaveTangent=0.6118, time=0.0281, value=0.9705}
         {arriveTangent=-1.4445, interpMode=RCIM_Cubic, leaveTangent=-1.4445, time=0.4, value=0.25}
         <at 0.7 0>
        ]}}})
    Scale RGB=&VectorFromFloat(Value=&FloatFromCurve(
      FloatCurve=DataInterface() {propertyValues {Curve {keys [
          <at 0 1 RCIM_Cubic>
          {arriveTangent=29.9839, interpMode=RCIM_Cubic, leaveTangent=29.9839, time=0.0201, value=1.7526}
          {arriveTangent=-1.1945, interpMode=RCIM_Cubic, leaveTangent=-1.1945, time=0.064, value=2.9193}
          {arriveTangent=-2.0506, interpMode=RCIM_Cubic, leaveTangent=-2.0506, time=0.5916, value=1.0699}
          <at 1 1>
         ]}}}
      Scale Curve=2
     ))
    ScaleA=true
    ScaleRGB=true
   )
   &SolveForcesAndVelocity(Clamp Velocity=false, Limit Acceleration=false)
  ]
  rendererClasses [&NiagaraMeshRendererProperties]
  renderers [{rendererClass=&NiagaraMeshRendererProperties, rendererIndex=0}]
 }
}
summary {
 systemName=NS_Bling_Van_Win_Burst
 emitters [
  {emitterName=GlowStart, rendererClasses [&NiagaraMeshRendererProperties]}
  {emitterName=FlashFlare, rendererClasses [&NiagaraSpriteRendererProperties]}
  {emitterName=Particle_Emitter, rendererClasses [&NiagaraMeshRendererProperties]}
  {emitterName=LOW_Sparks, simTarget=GPUComputeSim, rendererClasses [&NiagaraSpriteRendererProperties]}
  {emitterName=Sparks, simTarget=GPUComputeSim, rendererClasses [&NiagaraSpriteRendererProperties]}
  {emitterName=Wipe, rendererClasses [&NiagaraMeshRendererProperties]}
  {emitterName=MIDHIGH_Firework, rendererClasses [&NiagaraMeshRendererProperties]}
  {emitterName=Shimmers, rendererClasses [&NiagaraSpriteRendererProperties]}
 ]
 userVariables []
}
